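# Niagara idle cosmetic effect for the SummitReedMolt character body, emitting dark smoke from six skeletal sockets.
# Trail keeps six socket-following meshes, while SmokeDetailed lerps sprites toward a lagged socket position and SmokeBurst emits short red-to-black puffs.
# System reactivity curves and timeline drive spawning, opacity and material parameters.
# Spawn-group switches map particles to sockets and hide selected groups for ADS or hidden hands, with LOD and time-of-day adjustments.
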NiagaraSystem /BRCosmetics/Characters/Player/Male/Medium/Bodies/M_MED_SummitReedMolt/FX/NS_Character_SummitReedMolt_Peel_Idle
dependencies {
 usedDataInterfaces [
  &NiagaraDataInterfaceVectorCurve
  &NiagaraDataInterfaceCurve
  &NiagaraDataInterfaceVector2DCurve
  &NiagaraDataInterfaceSpriteRendererInfo
  &NiagaraDataInterfaceSocketReader
 ]
 usedDynamicInputs [
  &Vector2DFromCurve
  &RandomRangeVector2D
  &Divide_Float
  &InvertBool
  &SetBoolByFloatComparison
  &Abs_Float
  &VectorLength
  &SetBoolByIntComparison
 ]
 usedModules [
  &self:SystemSpawnScript.SystemScriptSource.SystemScriptGraph.NiagaraNodeAssignment_4.SetVariables_#1
  &SystemState
  /Game/Effects/Niagara/Modules/Logic/UpdateValueStateChange
  &Timeline
  &self:SystemSpawnScript.SystemScriptSource.SystemScriptGraph.NiagaraNodeAssignment_0.SetVariables_#2
  &self:SystemSpawnScript.SystemScriptSource.SystemScriptGraph.NiagaraNodeAssignment_3.SetVariables_#3
  &self:SystemSpawnScript.SystemScriptSource.SystemScriptGraph.NiagaraNodeAssignment_1.SetVariables_#4
  &self:SystemSpawnScript.SystemScriptSource.SystemScriptGraph.NiagaraNodeAssignment_2.SetVariables_#5
  &FN_Integrator
  &self:SystemSpawnScript.SystemScriptSource.SystemScriptGraph.NiagaraNodeAssignment_5.SetVariables_#6
 ]
}
emitters {
 SmokeBurst {
  emitterUpdateScript [
   &EmitterState(Life Cycle Mode=::NewEnumerator1, Loop Delay=0.4, Loop Duration=0.5, UseLoopDelay=false)
   &SpawnRate(
    Spawn Probability=$User.ReactivityOnOff
    SpawnRate=&Multiply_Float(
     A=&FloatFromCurve(
      FloatCurve=DataInterface() {propertyValues {Curve {keys [<at 0 1 RCIM_Constant>,<at 0.05 0 RCIM_Constant>]}}}
      CurveIndex=$System.Timeline.NormalizedTime
     )
     B=150
    )
    Use Spawn Probability=true
   )
  ]
  particleSpawnScript [
   &FN_SetSpawnGroup(Number of Spawn Groups=6)
   &InitializeParticle(
    Lifetime=&RandomRangeFloat(Minimum=0.7, Maximum=1)
    Mesh Scale Mode=::NewEnumerator3
    Sprite Rotation Mode=::NewEnumerator2
    Sprite Size Max=<100 100>
    Sprite Size Min=<70 70>
    Sprite Size Mode=::NewEnumerator2
    UsePositionOffset=true
    Write Material Random=true
   )
   &CameraOffset(Camera Offset Amount=30)
   &SocketLocation(
    SamplePosOffset=true
    Sampled Position Offset=&MergeHelper_VectorSwitch(0=<2 -2 0>, 1=<2 -2 0>, 2=<4 1 -5>, 3=<-4 -1 5>, 4=<4 -1.5 -6>, 5=<-4 3.5 5>, Index=$Particles.SpawnGroup, Num=::NewEnumerator4)
    Socket Index=&MergeHelper_IntSwitch(0=6, 1=7, 2=2, 3=3, 4=4, 5=5, Index=$Particles.SpawnGroup, Num=::NewEnumerator4)
    Socket Reader=$System.Socket Reader
    Socket Sampling Mode=::NewEnumerator3
    VelocitySpeedLimit=false
   ) {module=SocketLocation001}
   &AddVelocity(
    Module.Use Falloff Distance=false
    Override Seed=false
    Velocity Mode=::NewEnumerator1
    Velocity Origin=$Output.SocketLocation001.SampledPosition
    Velocity Speed=&RandomRangeFloat(Minimum=5, Maximum=25)
   )
  ]
  particleUpdateScript [
   &ParticleState()
   &SpriteRotationRate(Rotation Rate=&RandomRangeFloat(Minimum=-50, Maximum=50))
   &ScaleSpriteSize(
    Scale Sprite Size Mode=::NewEnumerator0
    Uniform Scale Factor=&FloatFromCurve(
     FloatCurve=DataInterface() {propertyValues {Curve {
        defaultValue=1
        keys [
         {arriveTangent=4.1785, interpMode=RCIM_Cubic, leaveTangent=4.1785, tangentMode=RCTM_User, time=0, value=1}
         {arriveTangent=0.6516, interpMode=RCIM_Cubic, leaveTangent=0.6516, tangentMode=RCTM_User, time=0.2042, value=1.7491}
         {arriveTangent=0.0224, interpMode=RCIM_Cubic, leaveTangent=0.0224, tangentMode=RCTM_User, time=1, value=2}
        ]
       }}}
     Scale Curve=0.08
    )
   )
   &SubUVAnimation(
    End Frame Range Override=30
    "Sprite Renderer "=DataInterface() {propertyValues {SpriteRenderer=&self:SmokeBurst_0.NiagaraSpriteRendererProperties_0}}
    Start Frame Range Override=&RandomRangeInt(Minimum=0, Maximum=5)
    UseEndFrame=true
    UseStartFrame=true
   )
   &ScaleColor(
    Scale Alpha=&Multiply_Float(
     A=&RandomRangeFloat(Minimum=1, Maximum=1)
     B=&FloatFromCurve(
      FloatCurve=DataInterface() {propertyValues {Curve {keys [
          {arriveTangent=0.9539, interpMode=RCIM_Cubic, leaveTangent=0.9539, tangentMode=RCTM_User, time=0, value=0.7}
          {interpMode=RCIM_Cubic, tangentMode=RCTM_User, time=0.4, value=1}
          {tangentMode=RCTM_User, time=1, value=0}
         ]}}}
      Scale Curve=&RandomRangeFloat(Minimum=1, Maximum=1.5)
     )
    )
    Scale RGB=&MakeVectorFromLinearColorRGB(Color=&Multiply_LinearColorByFloat(
      Linear Color=&LerpLinearColors(
       StartColor=<rgba 0.1 0 0 1>
       EndColor=&LerpLinearColors(
        StartColor=<rgba 0.016 0.016 0.016 1>
        EndColor=<rgba 0.04 0.04 0.04 1>
        LerpFactor=&UniformAOrBFloat(A=0, Evaluation Type=::NewEnumerator0, Recalculate Random Each Loop=true, Override Seed=false)
       )
       LerpFactor=&FloatFromCurve(FloatCurve=DataInterface() {propertyValues {Curve {keys [<at 0 0>,<at 0.2 1>]}}})
      )
      Float=&FNTimeOfDayScalar(Night Scale=0.6, Morning Scale=0.9, Evening Scale=0.8)
     ))
    ScaleA=true
    ScaleRGB=true
   )
   &ScaleColor(ScaleA=true, ScaleRGB=false) {module=ScaleColor001}
   &ScaleColor(
    Scale Alpha=&MakeCustomFloatFromBool(
     Bool=$User.bIsADS
     True Float=&MergeHelper_FloatSwitch(0=0, 1=1, 2=0, 3=1, 4=0, 5=1, Index=$Particles.SpawnGroup, Num=::NewEnumerator4)
     False Float=1
    )
    ScaleA=true
    ScaleRGB=false
   ) {enabled=false, module=ScaleColor003}
   &Drag(Drag=2, Use Linear Drag=true, Use Rotational Drag=true)
   &GravityForce(Gravity=&MakeVector(X=0, Y=0, Z=&RandomRangeFloat(Minimum=20, Maximum=40)))
   &InheritSourceMovement(
    Applied Position Scale=&VectorFromFloat(Value=&FloatFromCurve(FloatCurve=DataInterface() {propertyValues {Curve {keys [<at 0.5 1>,<at 1 0>]}}}))
    ApplyForce=false
    ApplyPosition=true
    ApplyPositionFalloff=false
   )
   &SolveForcesAndVelocity(Clamp Velocity=false, Limit Acceleration=false)
   &ScaleColor(
    Scale Alpha=&MakeCustomFloatFromBool(Bool=$System.LODVisibility, True Float=1, False Float=0)
    ScaleA=true
    ScaleRGB=false
   ) {module=ScaleColor004}
   &ScaleColor(
    Scale Alpha=&MakeCustomFloatFromBool(
     Bool=$User.bIsADS
     True Float=&MergeHelper_FloatSwitch(0=1, 1=1, 2=0, 3=1, 4=1, 5=1, Index=$Particles.SpawnGroup, Num=::NewEnumerator4)
     False Float=1
    )
    ScaleA=true
    ScaleRGB=false
   ) {module=ScaleColor005}
   &ScaleColor(
    Scale Alpha=&MakeCustomFloatFromBool(
     Bool=$User.bisHandHidden
     True Float=&MergeHelper_FloatSwitch(0=1, 1=1, 2=1, 3=1, 4=0, 5=0, Index=$Particles.SpawnGroup, Num=::NewEnumerator4)
     False Float=1
    )
    ScaleA=true
    ScaleRGB=false
   ) {module=ScaleColor002}
  ]
  rendererClasses [&NiagaraSpriteRendererProperties]
  renderers [{rendererClass=&NiagaraSpriteRendererProperties}]
 }
 SmokeDetailed {
  emitterUpdateScript [
   &EmitterState(Inactive Response=::NewEnumerator1, Life Cycle Mode=::NewEnumerator1, Loop Delay=0.4, Loop Duration=0.5, UseLoopDelay=false)
   &SpawnBurst_Instantaneous(
    Spawn Count=2
    Spawn Probability=&Multiply_Float(A=$System.ReactivityCurve, B=$User.ReactivityOnOff)
    Use Spawn Probability=true
    UseLoopCountLimit=false
   )
   &SpawnRate(
    Spawn Probability=$User.ReactivityOnOff
    SpawnRate=&Multiply_Float(A=10, B=$System.ReactivityCurve)
    Use Spawn Probability=true
   )
  ]
  particleSpawnScript [
   &FN_SetSpawnGroup(Number of Spawn Groups=6)
   &InitializeParticle(
    Color=<rgba 1 1 1 0.5>
    Lifetime=2.5
    Sprite Rotation Mode=::NewEnumerator2
    Sprite Size Max=<100 100>
    Sprite Size Min=<80 80>
    Sprite Size Mode=::NewEnumerator2
    UsePositionOffset=true
    Write Material Random=true
   )
   &CameraOffset(Camera Offset Amount=5)
  ]
  particleUpdateScript [
   &ParticleState()
   &SocketLocation(
    SamplePosOffset=true
    Sampled Position Offset=&MergeHelper_VectorSwitch(
     0=&AddVectorToPosition(
      Position=<0 0 0>
      Vector=&VectorFromCurve(
       VectorCurve=DataInterface() {propertyValues {XCurve {keys [<at 0 0>,<at 1 1>]}, YCurve {keys [<at 0 0>,<at 1 1>]}, ZCurve {keys [<at 0 0>,<at 1 1>]}}}
       Scale Curve=<0 -15 0>
      )
     )
     1=&AddVectorToPosition(
      Position=<0 0 0>
      Vector=&VectorFromCurve(
       VectorCurve=DataInterface() {propertyValues {XCurve {keys [<at 0 0>,<at 1 1>]}, YCurve {keys [<at 0 0>,<at 1 1>]}, ZCurve {keys [<at 0 0>,<at 1 1>]}}}
       Scale Curve=<0 -15 1>
      )
     )
     2=&AddVectorToPosition(
      Position=<0 0 0>
      Vector=&VectorFromCurve(
       VectorCurve=DataInterface() {propertyValues {XCurve {keys [<at 0 0>,<at 1 1>]}, YCurve {keys [<at 0 0>,<at 1 1>]}, ZCurve {keys [<at 0 0>,<at 1 1>]}}}
       Scale Curve=<5 0 -20>
      )
     )
     3=&AddVectorToPosition(
      Position=<0 0 0>
      Vector=&VectorFromCurve(
       VectorCurve=DataInterface() {propertyValues {XCurve {keys [<at 0 0>,<at 1 1>]}, YCurve {keys [<at 0 0>,<at 1 1>]}, ZCurve {keys [<at 0 0>,<at 1 1>]}}}
       Scale Curve=<-5 0 20>
      )
     )
     4=&AddVectorToPosition(
      Position=<4 -1.5 -6>
      Vector=&VectorFromCurve(
       VectorCurve=DataInterface() {propertyValues {XCurve {keys [<at 0 0>,<at 1 1>]}, YCurve {keys [<at 0 0>,<at 1 1>]}, ZCurve {keys [<at 0 0>,<at 1 1>]}}}
       Scale Curve=<20 -5 -20>
      )
     )
     5=&AddVectorToPosition(
      Position=<-4 3.5 5>
      Vector=&VectorFromCurve(
       VectorCurve=DataInterface() {propertyValues {XCurve {keys [<at 0 0>,<at 1 1>]}, YCurve {keys [<at 0 0>,<at 1 1>]}, ZCurve {keys [<at 0 0>,<at 1 1>]}}}
       Scale Curve=<-20 10 20>
      )
     )
     Index=$Particles.SpawnGroup
     Num=::NewEnumerator4
    )
    Socket Index=&MergeHelper_IntSwitch(0=6, 1=7, 2=2, 3=3, 4=4, 5=5, Index=$Particles.SpawnGroup, Num=::NewEnumerator4)
    Socket Reader=$System.Socket Reader
    Socket Sampling Mode=::NewEnumerator3
    VelocitySpeedLimit=false
   ) {module=SocketLocation001}
   &InterpolateOverTime(Rate Of Change=10, Target Vector Value:NiagaraPosition=$Output.SocketLocation001.SampledPosition, Variable Type=::NewEnumerator2)
   /SetVariables(Particles.Position=&LerpPosition(
     A=$Output.SocketLocation001.SampledPosition
     B=Linked() {linkedVariable {name=StackContext.InterpolateOverTime.InterpolateOverTime_V3_.Moving Average, type=Vector3f}}
     Alpha=&FloatFromCurve(FloatCurve=DataInterface() {propertyValues {Curve {keys [
          {interpMode=RCIM_Cubic, leaveTangent=1.2292, tangentMode=RCTM_Break, time=0.25, value=0}
          {interpMode=RCIM_Cubic, tangentMode=RCTM_Break, time=1, value=1}
         ]}}})
    )) {isSetParameters=true}
   &SpriteRotationRate(Rotation Rate=&RandomRangeFloat(Minimum=-50, Maximum=50))
   &ScaleSpriteSize(
    Scale Sprite Size Mode=::NewEnumerator0
    Uniform Scale Factor=&FloatFromCurve(
     FloatCurve=DataInterface() {propertyValues {Curve {
        defaultValue=1
        keys [
         {arriveTangent=-0.6693, interpMode=RCIM_Cubic, leaveTangent=-0.6693, tangentMode=RCTM_User, time=0, value=1}
         {arriveTangent=0.0224, interpMode=RCIM_Cubic, leaveTangent=0.0224, tangentMode=RCTM_User, time=1, value=0.35}
        ]
       }}}
     Scale Curve=0.2
    )
   ) {module=ScaleSpriteSize001}
   &SubUVAnimation(
    End Frame Range Override=30
    "Sprite Renderer "=DataInterface() {propertyValues {SpriteRenderer=&self:SmokeDetailed_0.NiagaraSpriteRendererProperties_0}}
    Start Frame Range Override=&RandomRangeInt(Minimum=0, Maximum=5)
    UseEndFrame=true
    UseStartFrame=true
   )
   &ScaleColor(
    Scale Alpha=&Multiply_Float(
     A=&RandomRangeFloat(Minimum=1, Maximum=1)
     B=&FloatFromCurve(
      FloatCurve=DataInterface() {propertyValues {Curve {keys [
          {arriveTangent=3.7485, interpMode=RCIM_Cubic, leaveTangent=3.7485, tangentMode=RCTM_User, time=0, value=0.35}
          {interpMode=RCIM_Cubic, tangentMode=RCTM_User, time=0.4, value=1}
          {tangentMode=RCTM_User, time=1, value=0}
         ]}}}
      Scale Curve=&FN_FrontEndFloatModifier(Gameplay=2, Front End=1.5)
     )
    )
    Scale RGB=&MakeVectorFromLinearColorRGB(Color=&Multiply_LinearColorByFloat(
      Linear Color=<rgba 0.016 0.016 0.016 1>
      Float=&FNTimeOfDayScalar(Night Scale=0.6, Morning Scale=0.9, Evening Scale=0.8)
     ))
    ScaleA=true
    ScaleRGB=true
   )
   &ScaleColor(Scale Alpha=$System.ReactivityCurveSprites, ScaleA=true, ScaleRGB=false) {module=ScaleColor001}
   &SolveForcesAndVelocity(Clamp Velocity=false, Limit Acceleration=false)
   /SetVariables(Particles.Position=&AddVectorToPosition(
     Position=$StackContext.Position
     Vector=&VectorFromCurve(
      VectorCurve=DataInterface() {propertyValues {
        XCurve {keys [{interpMode=RCIM_Cubic, tangentMode=RCTM_Break, time=0.3, value=0}, {interpMode=RCIM_Cubic, tangentMode=RCTM_Break, time=1, value=1}]}
        YCurve {keys [{interpMode=RCIM_Cubic, tangentMode=RCTM_Break, time=0.3, value=0}, {interpMode=RCIM_Cubic, tangentMode=RCTM_Break, time=1, value=1}]}
        ZCurve {keys [{interpMode=RCIM_Cubic, tangentMode=RCTM_Break, time=0.3, value=0}, {interpMode=RCIM_Cubic, tangentMode=RCTM_Break, time=1, value=1}]}
       }}
      Scale Curve=&MakeVector(
       X=0
       Y=0
       Z=&MergeHelper_FloatSwitch(0=4, 1=4, 2=4, 3=4, 4=7, 5=7, Index=$Particles.SpawnGroup, Num=::NewEnumerator4)
      )
     )
    )) {isSetParameters=true}
   &ScaleColor(
    Scale Alpha=&MakeCustomFloatFromBool(
     Bool=$User.bisHandHidden
     True Float=&MergeHelper_FloatSwitch(0=1, 1=1, 2=1, 3=1, 4=0, 5=0, Index=$Particles.SpawnGroup, Num=::NewEnumerator4)
     False Float=1
    )
    ScaleA=true
    ScaleRGB=false
   ) {module=ScaleColor002}
   &ScaleColor(
    Scale Alpha=&MakeCustomFloatFromBool(
     Bool=$User.bIsADS
     True Float=&MergeHelper_FloatSwitch(0=1, 1=1, 2=1, 3=1, 4=1, 5=1, Index=$Particles.SpawnGroup, Num=::NewEnumerator4)
     False Float=1
    )
    ScaleA=true
    ScaleRGB=false
   ) {module=ScaleColor003}
   &ScaleColor(
    Scale Alpha=&MakeCustomFloatFromBool(Bool=$System.LODVisibility, True Float=1, False Float=0)
    ScaleA=true
    ScaleRGB=false
   ) {module=ScaleColor004}
  ]
  rendererClasses [&NiagaraSpriteRendererProperties]
  renderers [{rendererClass=&NiagaraSpriteRendererProperties}]
 }
 Trail {
  emitterUpdateScript [
   &EmitterState(Life Cycle Mode=::NewEnumerator1, Loop Behavior=::NewEnumerator1, Loop Duration Mode=::NewEnumerator1, UseLoopDelay=false)
   &SpawnBurst_Instantaneous(Spawn Count=6, Spawn Probability=$User.ReactivityOnOff, Use Spawn Probability=true, UseLoopCountLimit=false)
  ]
  particleSpawnScript [
   &FN_SetSpawnGroup(Number of Spawn Groups=6)
   &InitializeParticle(Color Mode=::NewEnumerator0, Lifetime=5, Mesh Scale Mode=::NewEnumerator3, UsePositionOffset=false, Write Material Random=true)
   /SetVariables(Particles.VisibilityTag=&MergeHelper_IntSwitch(0=0, 1=0, 2=1, 3=1, 4=2, 5=2, Index=$Particles.SpawnGroup, Num=::NewEnumerator4)) {isSetParameters=true}
  ]
  particleUpdateScript [
   &ParticleState(Kill Particles When Lifetime Has Elapsed=false)
   &SocketLocation(
    SamplePosOffset=true
    Sampled Position Offset=&MergeHelper_VectorSwitch(0=<2 -1 0>, 1=<2 -1 0>, 2=<2 0 -1>, 3=<-2 0 1>, 4=<0 -2 0>, 5=<0 -2 0>, Index=$Particles.SpawnGroup, Num=::NewEnumerator4)
    Socket Index=&MergeHelper_IntSwitch(0=6, 1=7, 2=2, 3=3, 4=4, 5=5, Index=$Particles.SpawnGroup, Num=::NewEnumerator4)
    Socket Reader=$System.Socket Reader
    Socket Sampling Mode=::NewEnumerator3
    VelocitySpeedLimit=false
   )
   &InterpolateOverTime(
    Rate Of Change:NiagaraFloat=&RandomRangeFloat(Minimum=5, Maximum=6)
    Target Vector Value:NiagaraPosition=$Particles.Position
    Variable Type=::NewEnumerator2
   )
   /SetVariables(Particles.Interpolate DY:Vector4f=&MakeFloat4FromVectorAndFloat(
     "Vector (RGB)"=&MakeVector(
      X=&MakeFloatFromVector(Vector=$Particles.InterpolateOverTime.InterpolateOverTime_V3_.Moving Average, Channel=::NewEnumerator0)
      Y=&MakeFloatFromVector(Vector=$Particles.InterpolateOverTime.InterpolateOverTime_V3_.Moving Average, Channel=::NewEnumerator1)
      Z=&MakeFloatFromVector(Vector=Linked() {linkedVariable {name=Output.SocketLocation.SampledPosition, type=NiagaraPosition}}, Channel=::NewEnumerator2)
     )
     "Float (Alpha)"=1
    )) {isSetParameters=true}
   &DynamicMaterialParameters(
    Index 0 Param 1=1.4
    Index 0 Param 2=2.3
    Index 0 Param 3=3
    Index 0 Param 4=&Add_Float(
     A=&MergeHelper_FloatSwitch(0=0, 1=0.4, 2=0, 3=0.6, 4=0, 5=0.3, Index=$Particles.SpawnGroup, Num=::NewEnumerator4)
     B=&Multiply_Float(A=$System.FN_Integrator002.Value, B=0.5)
    )
    Index 2 Param 1=2.2
    Index 2 Param 2=3
    Index 2 Param 3=1.5
    Index 3 Param 4=&Lerp_Float(A=-1, B=1, Alpha=$System.ReactivityCurve)
    Index 3 Vector 1:Vector3f=&MakeVectorFromLinearColorRGB(Color=&Multiply_LinearColorByFloat(
      Linear Color=<rgba 1 1 1 1>
      Float=&FN_FrontEndFloatModifier(
       Gameplay=&FNTimeOfDayScalar(Night Scale=0.4, Morning Scale=0.9, Evening Scale=0.8)
       Front End=0.6
      )
     ))
    Module.Index0Float0Write=true
    Module.Index0Float1Write=true
    Module.Index0Float2Write=true
    Module.Index0Float3Write=true
    Module.Index2Float0Write=true
    Module.Index2Float1Write=true
    Module.Index2Float2Write=true
    Module.Index2Float3Write=true
    Module.Index3Float3Write=true
    Module.Index3VectorWrite=true
    Write Parameter Index 2=::NewEnumerator1
    Write Parameter Index 3=::NewEnumerator3
   )
   &ScaleColor(
    Scale Alpha=13
    Scale RGB=&MakeVectorFromLinearColorRGB(Color=<rgba 0.016 0.016 0.016 1>)
    ScaleA=true
    ScaleRGB=true
   )
   &ScaleMeshSize(Scale Factor=&VectorFromFloat(Value=&MergeHelper_FloatSwitch(0=0.5, 1=0.5, 2=0.6, 3=0.6, 4=0.8, 5=0.8, Index=$StackContext.SpawnGroup, Num=::NewEnumerator4)))
   /SetVariables(Particles.MeshOrientation=&MultiplyQuaternion(
     Quaternion A=$Particles.MeshOrientation
     Quaternion B=&MakeQuaternion(
      Quaternion From=0
      Axis=&MergeHelper_VectorSwitch(0=<0 1 0>, 1=<0 1 0>, 2=<0 1 0>, 3=<0 1 0>, 4=<0 0 0>, 5=<0.5 1 0>, Index=$Particles.SpawnGroup, Num=::NewEnumerator4)
      Angle=&MergeHelper_FloatSwitch(0=0, 1=0, 2=0, 3=180, 4=90, 5=180, Index=$Particles.SpawnGroup, Num=::NewEnumerator4)
      From=?
      To=?
      Angle Type=::NewEnumerator0
      Normalize Vectors=true
     )
    )) {isSetParameters=true}
   &ScaleColor(
    Scale Alpha=&Multiply_Float(A=0.6, B=&Lerp_Float(A=1, B=0.6, Alpha=$System.InterpolateOverTime.InterpolateOverTime_Float.Moving Average))
    ScaleA=true
    ScaleRGB=true
   ) {module=ScaleColor001}
   &ScaleColor(
    Scale Alpha=&MakeCustomFloatFromBool(
     Bool=$User.bisHandHidden
     True Float=&MergeHelper_FloatSwitch(0=1, 1=1, 2=1, 3=1, 4=0, 5=0, Index=$Particles.SpawnGroup, Num=::NewEnumerator4)
     False Float=1
    )
    ScaleA=true
    ScaleRGB=false
   ) {module=ScaleColor002}
   &ScaleColor(
    Scale Alpha=&MakeCustomFloatFromBool(
     Bool=$User.bIsADS
     True Float=&MergeHelper_FloatSwitch(0=1, 1=1, 2=0, 3=1, 4=1, 5=1, Index=$Particles.SpawnGroup, Num=::NewEnumerator4)
     False Float=1
    )
    ScaleA=true
    ScaleRGB=false
   ) {module=ScaleColor003}
   &ScaleColor(
    Scale Alpha=&MakeCustomFloatFromBool(Bool=$System.LODVisibility, True Float=1, False Float=0)
    ScaleA=true
    ScaleRGB=false
   ) {module=ScaleColor004}
  ]
  rendererClasses [&NiagaraMeshRendererProperties]
  renderers [
   {rendererClass=&NiagaraMeshRendererProperties, rendererIndex=0}
   {rendererClass=&NiagaraMeshRendererProperties, rendererIndex=1}
   {rendererClass=&NiagaraMeshRendererProperties, rendererIndex=2}
  ]
 }
}
summary {
 systemName=NS_Character_SummitReedMolt_Peel_Idle
 emitters [
  {emitterName=Trail, rendererClasses [&NiagaraMeshRendererProperties]}
  {emitterName=SmokeDetailed, rendererClasses [&NiagaraSpriteRendererProperties]}
  {emitterName=SmokeBurst, rendererClasses [&NiagaraSpriteRendererProperties]}
 ]
 userVariables [
  User.bIsADS {type=NiagaraBool}
  User.bIsFrontEnd {type=NiagaraBool}
  User.bIsFrontEndPreview {type=NiagaraBool}
  User.bisHandHidden {type=NiagaraBool}
  User.DebugReactivity {type=NiagaraBool}
  User.ReactivityCounter {type=NiagaraInt32}
  User.ReactivityDuration {type=NiagaraFloat, defaultValue=6.5, description=""}
  User.ReactivityOnOff {type=NiagaraFloat, defaultValue=1, description=""}
  User.SkeletalMeshCurrentLOD {type=NiagaraInt32}
 ]
}
