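# Physics asset for the medium male SummitReedMolt cosmetic body, covering only the lower body and lower spine, apparently to drive swinging thigh bags.
# Capsules wrap pelvis, two spine bones, thighs and calves, with small ones on groin and bag deform bones and two per dyn_bag bone.
# Deform bones hang off the thighs with free angular constraints, as does the spine chain from the pelvis.
# Each dyn_bag joins its deform bone with tighter limits than the legs.
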PhysicsAsset /BRCosmetics/Characters/Player/Male/Medium/Bodies/M_MED_SummitReedMolt/Meshes/M_MED_SummitReedMolt_Physics
bodies {
 calf_l [{shapeType=Capsule, length=33.3824, radius=8.8195, center=<-18.3539 -1.1057 0>, rotation=<rot 90 179.9999 0>}]
 calf_r [{shapeType=Capsule, length=33.3824, radius=8.8195, center=<18.3539 1.1057 0>, rotation=<rot 90 -0.0001 0>}]
 deform_bag_l [{shapeType=Capsule, length=0.505, radius=0.505, shapeName=deform_bag_l_capsule, rotation=<rot 0 -90 0>}]
 deform_bag_r [{shapeType=Capsule, length=0.505, radius=0.505, shapeName=deform_bag_r_capsule, rotation=<rot 0 -90 0>}]
 deform_groin_l [{
   shapeType=Capsule
   length=3.1484
   radius=1.0417
   shapeName=deform_groin_l_capsule
   center=<-2.9078 1.3288 2.1295>
   rotation=<rot 8.7436 -90 -9.9943>
  }]
 deform_groin_r [{shapeType=Capsule, length=0.505, radius=0.505, shapeName=deform_groin_r_capsule, rotation=<rot 0 -90 0>}]
 dyn_bag_l [
  {
   shapeType=Capsule
   length=6.03
   radius=3.1973
   shapeName=dyn_bag_l_capsule
   center=<-0.8441 -2.9242 0.7315>
   rotation=<rot -0.0001 90.3637 5.3475>
  }
  {shapeType=Capsule, length=5.5972, radius=3.59, shapeName=dyn_bag_l_capsule, center=<5.6514 -2.8759 0>, rotation=<rot 0 0 -180>}
 ]
 dyn_bag_r [
  {
   shapeType=Capsule
   length=6.03
   radius=3.1973
   shapeName=dyn_bag_r_capsule
   center=<-0.8441 -2.9242 -0.7315>
   rotation=<rot -0.0001 90.3637 -174.6525>
  }
  {shapeType=Capsule, length=5.5972, radius=3.59, shapeName=dyn_bag_r_capsule, center=<5.6514 -2.8759 0>}
 ]
 pelvis [{shapeType=Capsule, length=5.9573, radius=11.3133}]
 spine_01 [{shapeType=Capsule, length=5.865, radius=10.6758, shapeName=spine_01_capsule}]
 spine_02 [{shapeType=Capsule, length=4.6452, radius=10.6758, shapeName=spine_02_capsule}]
 thigh_l [{shapeType=Capsule, length=37.5314, radius=9.1086, center=<-18.3539 0 1.9986>, rotation=<rot 90 180 0>}]
 thigh_r [{shapeType=Capsule, length=37.5314, radius=9.1086, center=<18.3539 0 -1.9986>, rotation=<rot 90 0 0>}]
}
constraints [
 {bone1Name=thigh_l, bone2Name=pelvis, swing1LimitDegrees=45, swing2LimitDegrees=45, twistLimitDegrees=45}
 {bone1Name=calf_l, bone2Name=thigh_l, swing1LimitDegrees=45, swing2LimitDegrees=45, twistLimitDegrees=45}
 {bone1Name=thigh_r, bone2Name=pelvis, swing1LimitDegrees=45, swing2LimitDegrees=45, twistLimitDegrees=45}
 {bone1Name=calf_r, bone2Name=thigh_r, swing1LimitDegrees=45, swing2LimitDegrees=45, twistLimitDegrees=45}
 {
  bone1Name=spine_01
  bone2Name=pelvis
  swing1LimitDegrees=45
  swing1Motion=Free
  swing2LimitDegrees=45
  swing2Motion=Free
  twistLimitDegrees=45
  twistMotion=Free
 }
 {
  bone1Name=spine_02
  bone2Name=spine_01
  swing1LimitDegrees=45
  swing1Motion=Free
  swing2LimitDegrees=45
  swing2Motion=Free
  twistLimitDegrees=45
  twistMotion=Free
 }
 {
  bone1Name=deform_groin_l
  bone2Name=thigh_l
  swing1LimitDegrees=45
  swing1Motion=Free
  swing2LimitDegrees=45
  swing2Motion=Free
  twistLimitDegrees=45
  twistMotion=Free
 }
 {
  bone1Name=deform_groin_r
  bone2Name=thigh_r
  swing1LimitDegrees=45
  swing1Motion=Free
  swing2LimitDegrees=45
  swing2Motion=Free
  twistLimitDegrees=45
  twistMotion=Free
 }
 {
  bone1Name=deform_bag_l
  bone2Name=thigh_l
  swing1LimitDegrees=45
  swing1Motion=Free
  swing2LimitDegrees=45
  swing2Motion=Free
  twistLimitDegrees=45
  twistMotion=Free
 }
 {bone1Name=dyn_bag_l, bone2Name=deform_bag_l, swing1LimitDegrees=35, swing2LimitDegrees=35, twistLimitDegrees=35}
 {
  bone1Name=deform_bag_r
  bone2Name=thigh_r
  swing1LimitDegrees=45
  swing1Motion=Free
  swing2LimitDegrees=45
  swing2Motion=Free
  twistLimitDegrees=45
  twistMotion=Free
 }
 {bone1Name=dyn_bag_r, bone2Name=deform_bag_r, swing1LimitDegrees=35, swing2LimitDegrees=35, twistLimitDegrees=35}
]
