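# Physics asset for the SummitReedMolt male medium cosmetic's hat face-accessory part, covering secondary motion of hair, bandana and shoulder pads.
# Mostly capsule bodies cover upper spine, neck, head, clavicles and upper arms, with dyn_ chains hanging off deform_ anchors on head and upper arms.
# Hair and bandana chains have angular limits, while body, anchor and shoulder-pad joints swing freely.
# Cross-constraints tie neighboring hair strands together and back hair to bandana.
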PhysicsAsset /BRCosmetics/Characters/Player/Male/Medium/Bodies/M_MED_SummitReedMolt/Meshes/Parts/M_MED_SummitReedMolt_FaceAcc_Hat_Physics
bodies {
 clavicle_l [{
   shapeType=Capsule
   length=12.2989
   radius=9.3042
   shapeName=clavicle_l_capsule
   center=<6.4453 -1.3287 -0.8523>
   rotation=<rot 73.125 -0.0027 179.9971>
  }]
 clavicle_r [{
   shapeType=Capsule
   length=12.2989
   radius=9.3042
   shapeName=clavicle_r_capsule
   center=<-6.4453 1.3287 0.8523>
   rotation=<rot 73.125 -0.0027 -0.0029>
  }]
 deform_bandana [{shapeType=Capsule, length=2, radius=2, shapeName=deform_bandana_capsule}]
 deform_hair_b_l [{shapeType=Capsule, length=0.505, radius=0.505, shapeName=deform_hair_b_l_capsule, center=<-2.8258 0 0.1744>, rotation=<rot 0 -90 0>}]
 deform_hair_b_r [{shapeType=Capsule, length=0.505, radius=0.505, shapeName=deform_hair_b_r_capsule, center=<-2.8258 0 0.1744>, rotation=<rot 0 -90 0>}]
 deform_hair_bk [{shapeType=Capsule, length=0.505, radius=0.505, shapeName=deform_hair_bk_capsule, center=<-2.8258 0 0.1744>, rotation=<rot 0 -90 0>}]
 deform_hair_f_l [{shapeType=Capsule, length=3, radius=2, shapeName=deform_hair_f_l_capsule, center=<1.4809 -0.8316 0>, rotation=<rot 90 0 0>}]
 deform_hair_f_r [{shapeType=Capsule, length=3, radius=2, shapeName=deform_hair_f_r_capsule, center=<1.4809 -0.8316 0>, rotation=<rot 90 0 0>}]
 deform_hair_m_l [{shapeType=Capsule, length=0.505, radius=0.505, shapeName=deform_hair_m_l_capsule, rotation=<rot 0 -90 0>}]
 deform_hair_m_r [{shapeType=Capsule, length=0.505, radius=0.505, shapeName=deform_hair_m_r_capsule, rotation=<rot 0 -90 0>}]
 deform_hair_mb_l [{shapeType=Capsule, length=5, radius=3, shapeName=deform_hair_mb_l_capsule, center=<4.7455 -2.9484 0>}]
 deform_hair_mb_r [{shapeType=Capsule, length=5, radius=3, shapeName=deform_hair_mb_r_capsule, center=<4.7455 -2.9484 0>}]
 deform_shoulderpad_l [{shapeType=Capsule, length=0.1, radius=0.1, shapeName=deform_shoulderpad_l_capsule}]
 deform_shoulderpad_r [{shapeType=Capsule, length=0.1, radius=0.1, shapeName=deform_shoulderpad_r_capsule}]
 dyn_bandana [{shapeType=Capsule, length=2, radius=2, shapeName=dyn_bandana_capsule}]
 dyn_bandana_l [{shapeType=Capsule, length=2, radius=2, shapeName=dyn_bandana_l_capsule, center=<2.3379 -0.6243 0>, rotation=<rot 90 0 0>}]
 dyn_bandana_r [{shapeType=Capsule, length=2, radius=2, shapeName=dyn_bandana_r_capsule, center=<2.3379 -0.6243 0>, rotation=<rot 90 0 0>}]
 dyn_hair_1_b_l [{shapeType=Capsule, length=6, radius=3.3, shapeName=dyn_hair_1_b_l_capsule, center=<6.5087 -4.0553 0.1744>}]
 dyn_hair_1_b_r [
  {shapeType=Capsule, length=6, radius=3.3, shapeName=dyn_hair_1_b_r_capsule, center=<6.737 -3.9676 0.1744>}
  {
   shapeType=Capsule
   length=1.0419
   radius=1.75
   shapeName=dyn_hair_1_b_r_capsule
   center=<3.2354 -1.5767 -2.727>
   rotation=<rot 56.864 0.0001 20.1417>
  }
 ]
 dyn_hair_1_bk [{shapeType=Capsule, length=12.5228, radius=3.3, shapeName=dyn_hair_1_bk_capsule, center=<6.7633 -2.5622 0.1744>}]
 dyn_hair_1_f_l [{shapeType=Capsule, length=3, radius=2, shapeName=dyn_hair_1_f_l_capsule, center=<1.4809 -0.8316 0>, rotation=<rot 90 0 0>}]
 dyn_hair_1_f_r [{shapeType=Capsule, length=3, radius=2, shapeName=dyn_hair_1_f_r_capsule, center=<1.4809 -0.8316 0>, rotation=<rot 90 0 0>}]
 dyn_hair_1_m_l [{shapeType=Capsule, length=4.7849, radius=2.5435, shapeName=dyn_hair_1_m_l_capsule, center=<2.1067 -2.8478 0>, rotation=<rot 90 154.6875 0>}]
 dyn_hair_1_m_r [{shapeType=Capsule, length=4.7849, radius=2.5435, shapeName=dyn_hair_1_m_r_capsule, center=<1.7008 -2.9109 0>, rotation=<rot 90 -25.3125 0>}]
 dyn_hair_1_mb_l [{shapeType=Capsule, length=5, radius=3, shapeName=dyn_hair_1_mb_l_capsule, center=<2.7863 -2.9484 0>}]
 dyn_hair_1_mb_r [{shapeType=Capsule, length=3.4957, radius=3.25, shapeName=dyn_hair_1_mb_r_capsule, center=<1.6582 -3.4709 0>, rotation=<rot 90 0 0>}]
 dyn_hair_2_f_l [{shapeType=Capsule, length=5.8503, radius=1.5631, shapeName=dyn_hair_2_f_l_capsule, center=<3.4139 -0.8316 0>, rotation=<rot 90 0 0>}]
 dyn_hair_2_f_r [{shapeType=Capsule, length=5.8503, radius=1.5631, shapeName=dyn_hair_2_f_r_capsule, center=<3.4139 -0.8316 0>, rotation=<rot 90 0 0>}]
 dyn_hair_3_f_l [{shapeType=Capsule, length=5.8503, radius=1.5631, shapeName=dyn_hair_3_f_l_capsule, center=<3.4139 -0.8316 0>, rotation=<rot 90 0 0>}]
 dyn_hair_3_f_r [{shapeType=Capsule, length=5.8503, radius=1.5631, shapeName=dyn_hair_3_f_r_capsule, center=<3.4139 -0.8316 0>, rotation=<rot 90 0 0>}]
 dyn_hair_b_l [
  {shapeType=Capsule, length=6, radius=3.3, shapeName=dyn_hair_b_l_capsule, center=<2.6433 -2.6009 0.1744>}
  {shapeType=Capsule, length=6, radius=3, shapeName=dyn_hair_b_l_capsule, center=<8.5091 -3.2135 0>}
 ]
 dyn_hair_b_r [
  {shapeType=Capsule, length=6, radius=3.3, shapeName=dyn_hair_b_r_capsule, center=<2.6433 -2.6009 0.1744>}
  {shapeType=Capsule, length=6.2528, radius=3, shapeName=dyn_hair_b_r_capsule, center=<8.3238 -2.4103 0>}
 ]
 dyn_hair_bk [{shapeType=Capsule, length=7.6373, radius=3.3, shapeName=dyn_hair_bk_capsule, center=<8.238 -2.6492 0>}]
 dyn_hair_f_l [{shapeType=Capsule, length=3, radius=2, shapeName=dyn_hair_f_l_capsule, center=<1.4809 -1.7551 0>, rotation=<rot 90 2.8125 0>}]
 dyn_hair_f_r [{shapeType=Capsule, length=3, radius=2, shapeName=dyn_hair_f_r_capsule, center=<1.4809 -0.8316 0>, rotation=<rot 90 0 0>}]
 dyn_hair_m_l [{shapeType=Capsule, length=5.5247, radius=2.7167, shapeName=dyn_hair_m_l_capsule, center=<2.512 -1.3844 0>, rotation=<rot 90 180 0>}]
 dyn_hair_m_r [{shapeType=Capsule, length=5.5247, radius=2.7167, shapeName=dyn_hair_m_r_capsule, center=<2.512 -1.3844 0>, rotation=<rot 90 0 0>}]
 dyn_hair_mb_l [{shapeType=Capsule, length=5, radius=3, shapeName=dyn_hair_mb_l_capsule, center=<4.7455 -2.9484 0>}]
 dyn_hair_mb_r [{shapeType=Capsule, length=5, radius=3, shapeName=dyn_hair_mb_r_capsule, center=<4.7455 -2.1652 -0.7074>, rotation=<rot 90 0 0>}]
 dyn_shoulderpad_1_l [{shapeType=Capsule, length=1.9075, radius=6.6783, shapeName=dyn_shoulderpad_1_l_capsule, center=<4.0932 4.7601 0>, rotation=<rot 0 0 -180>}]
 dyn_shoulderpad_1_r [
  {shapeType=Capsule, length=8.4075, radius=4.1783, shapeName=dyn_shoulderpad_1_r_capsule, center=<1.6992 4.8418 -0.1606>}
  {
   shapeType=Capsule
   length=5.25
   radius=4.25
   shapeName=dyn_shoulderpad_1_r_capsule
   center=<2.6712 2.4538 -0.0375>
   rotation=<rot 0 0 0.3708>
  }
 ]
 dyn_shoulderpad_l [{shapeType=Capsule, length=0.1, radius=0.1, shapeName=dyn_shoulderpad_l_capsule}]
 dyn_shoulderpad_r [{shapeType=Capsule, length=0.1, radius=0.1, shapeName=dyn_shoulderpad_r_capsule}]
 head [{shapeType=Capsule, length=7.479, radius=6.7046, center=<3.6218 2.0604 0>, rotation=<rot 90 11.2498 0>}]
 neck_01 [
  {shapeType=Sphere, radius=8, shapeName=neck_01_sphere, center=<-0.1542 4.2701 0>}
  {shapeType=Capsule, length=3.7826, radius=6.1042, shapeName=neck_01_capsule, center=<2.5527 -1.5399 0>, rotation=<rot 90 0 0>}
 ]
 neck_02 [{shapeType=Capsule, length=4.8868, radius=5, shapeName=neck_02_capsule, center=<2.5527 0 0>, rotation=<rot 90 0 0>}]
 spine_04 [{shapeType=Capsule, length=4.5, radius=15.25, shapeName=spine_04_capsule, center=<6.6147 -1.9268 0>}]
 spine_05 [{shapeType=Capsule, length=10.3452, radius=13.0539, shapeName=spine_05_capsule, center=<-2.1212 1.9013 0>, rotation=<rot 90 0 0>}]
 upperarm_l [{shapeType=Capsule, length=16.8974, radius=6.2362, shapeName=upperarm_l_capsule, center=<6.6094 0 0>, rotation=<rot 90 180 0>}]
 upperarm_r [{shapeType=Capsule, length=16.8974, radius=6.2362, shapeName=upperarm_r_capsule, center=<-6.6094 0 0>, rotation=<rot 90 0 0>}]
}
constraints [
 {
  bone1Name=head
  bone2Name=neck_02
  swing1LimitDegrees=45
  swing1Motion=Free
  swing2LimitDegrees=45
  swing2Motion=Free
  twistLimitDegrees=45
  twistMotion=Free
 }
 {
  bone1Name=neck_02
  bone2Name=neck_01
  swing1LimitDegrees=45
  swing1Motion=Free
  swing2LimitDegrees=45
  swing2Motion=Free
  twistLimitDegrees=45
  twistMotion=Free
 }
 {
  bone1Name=neck_01
  bone2Name=spine_05
  swing1LimitDegrees=45
  swing1Motion=Free
  swing2LimitDegrees=45
  swing2Motion=Free
  twistLimitDegrees=45
  twistMotion=Free
 }
 {
  bone1Name=clavicle_l
  bone2Name=spine_05
  swing1LimitDegrees=45
  swing1Motion=Free
  swing2LimitDegrees=45
  swing2Motion=Free
  twistLimitDegrees=45
  twistMotion=Free
 }
 {
  bone1Name=clavicle_r
  bone2Name=spine_05
  swing1LimitDegrees=45
  swing1Motion=Free
  swing2LimitDegrees=45
  swing2Motion=Free
  twistLimitDegrees=45
  twistMotion=Free
 }
 {
  bone1Name=upperarm_l
  bone2Name=clavicle_l
  swing1LimitDegrees=45
  swing1Motion=Free
  swing2LimitDegrees=45
  swing2Motion=Free
  twistLimitDegrees=45
  twistMotion=Free
 }
 {
  bone1Name=upperarm_r
  bone2Name=clavicle_r
  swing1LimitDegrees=45
  swing1Motion=Free
  swing2LimitDegrees=45
  swing2Motion=Free
  twistLimitDegrees=45
  twistMotion=Free
 }
 {bone1Name=dyn_hair_b_l, bone2Name=deform_hair_b_l, swing1LimitDegrees=60, swing2LimitDegrees=10, twistLimitDegrees=5}
 {bone1Name=dyn_hair_1_b_l, bone2Name=dyn_hair_b_l, swing1LimitDegrees=60, swing2LimitDegrees=15, twistLimitDegrees=5}
 {bone1Name=dyn_hair_b_r, bone2Name=deform_hair_b_r, swing1LimitDegrees=60, swing2LimitDegrees=20, twistLimitDegrees=20}
 {bone1Name=dyn_hair_1_b_r, bone2Name=dyn_hair_b_r, swing1LimitDegrees=60, swing2LimitDegrees=15, twistLimitDegrees=5}
 {bone1Name=dyn_hair_bk, bone2Name=deform_hair_bk, swing1LimitDegrees=60, swing2LimitDegrees=10, twistLimitDegrees=5}
 {bone1Name=dyn_hair_1_bk, bone2Name=dyn_hair_bk, swing1LimitDegrees=60, swing2LimitDegrees=15, twistLimitDegrees=5}
 {bone1Name=dyn_hair_m_l, bone2Name=deform_hair_m_l, swing1LimitDegrees=20, swing2LimitDegrees=15, twistLimitDegrees=5}
 {bone1Name=dyn_hair_1_m_l, bone2Name=dyn_hair_m_l, swing1LimitDegrees=30, swing2LimitDegrees=30, twistLimitDegrees=30}
 {bone1Name=dyn_hair_m_r, bone2Name=deform_hair_m_r, swing1LimitDegrees=20, swing2LimitDegrees=35, twistLimitDegrees=5}
 {bone1Name=dyn_hair_1_m_r, bone2Name=dyn_hair_m_r, swing1LimitDegrees=30, swing2LimitDegrees=30, twistLimitDegrees=30}
 {
  bone1Name=deform_hair_b_l
  bone2Name=head
  swing1LimitDegrees=45
  swing1Motion=Free
  swing2LimitDegrees=45
  swing2Motion=Free
  twistLimitDegrees=45
  twistMotion=Free
 }
 {
  bone1Name=deform_hair_b_r
  bone2Name=head
  swing1LimitDegrees=45
  swing1Motion=Free
  swing2LimitDegrees=45
  swing2Motion=Free
  twistLimitDegrees=45
  twistMotion=Free
 }
 {
  bone1Name=deform_hair_bk
  bone2Name=head
  swing1LimitDegrees=45
  swing1Motion=Free
  swing2LimitDegrees=45
  swing2Motion=Free
  twistLimitDegrees=45
  twistMotion=Free
 }
 {
  bone1Name=deform_hair_m_l
  bone2Name=head
  swing1LimitDegrees=45
  swing1Motion=Free
  swing2LimitDegrees=45
  swing2Motion=Free
  twistLimitDegrees=45
  twistMotion=Free
 }
 {
  bone1Name=deform_hair_m_r
  bone2Name=head
  swing1LimitDegrees=45
  swing1Motion=Free
  swing2LimitDegrees=45
  swing2Motion=Free
  twistLimitDegrees=45
  twistMotion=Free
 }
 {bone1Name=dyn_hair_f_l, bone2Name=deform_hair_f_l, swing1LimitDegrees=35, swing2LimitDegrees=25, twistLimitDegrees=10}
 {bone1Name=dyn_hair_1_f_l, bone2Name=dyn_hair_f_l, swing1LimitDegrees=35, swing2LimitDegrees=25, twistLimitDegrees=10}
 {bone1Name=dyn_hair_2_f_l, bone2Name=dyn_hair_1_f_l, swing1LimitDegrees=15, swing2LimitDegrees=15, twistLimitDegrees=10}
 {bone1Name=dyn_hair_3_f_l, bone2Name=dyn_hair_1_f_l, swing1LimitDegrees=15, swing2LimitDegrees=15, twistLimitDegrees=10}
 {bone1Name=dyn_hair_f_r, bone2Name=deform_hair_f_r, swing1LimitDegrees=35, swing2LimitDegrees=25, twistLimitDegrees=10}
 {bone1Name=dyn_hair_1_f_r, bone2Name=dyn_hair_f_r, swing1LimitDegrees=35, swing2LimitDegrees=25, twistLimitDegrees=10}
 {bone1Name=dyn_hair_2_f_r, bone2Name=dyn_hair_1_f_r, swing1LimitDegrees=15, swing2LimitDegrees=15, twistLimitDegrees=10}
 {bone1Name=dyn_hair_3_f_r, bone2Name=dyn_hair_1_f_r, swing1LimitDegrees=15, swing2LimitDegrees=15, twistLimitDegrees=10}
 {
  bone1Name=deform_hair_f_r
  bone2Name=head
  swing1LimitDegrees=45
  swing1Motion=Free
  swing2LimitDegrees=45
  swing2Motion=Free
  twistLimitDegrees=45
  twistMotion=Free
 }
 {
  bone1Name=deform_hair_f_l
  bone2Name=head
  swing1LimitDegrees=45
  swing1Motion=Free
  swing2LimitDegrees=45
  swing2Motion=Free
  twistLimitDegrees=45
  twistMotion=Free
 }
 {bone1Name=dyn_bandana, bone2Name=deform_bandana, swing1LimitDegrees=60, swing2LimitDegrees=4, twistLimitDegrees=4}
 {bone1Name=dyn_bandana_l, bone2Name=dyn_bandana, swing1LimitDegrees=15, swing2LimitDegrees=15, twistLimitDegrees=3}
 {bone1Name=dyn_bandana_r, bone2Name=dyn_bandana, swing1LimitDegrees=15, swing2LimitDegrees=15, twistLimitDegrees=3}
 {
  bone1Name=deform_bandana
  bone2Name=head
  swing1LimitDegrees=45
  swing1Motion=Free
  swing2LimitDegrees=45
  swing2Motion=Free
  twistLimitDegrees=45
  twistMotion=Free
 }
 {
  bone1Name=deform_shoulderpad_l
  bone2Name=upperarm_l
  swing1LimitDegrees=45
  swing1Motion=Free
  swing2LimitDegrees=45
  swing2Motion=Free
  twistLimitDegrees=45
  twistMotion=Free
 }
 {
  bone1Name=dyn_shoulderpad_l
  bone2Name=deform_shoulderpad_l
  swing1LimitDegrees=45
  swing1Motion=Free
  swing2LimitDegrees=45
  swing2Motion=Free
  twistLimitDegrees=45
  twistMotion=Free
 }
 {
  bone1Name=dyn_shoulderpad_1_l
  bone2Name=dyn_shoulderpad_l
  swing1LimitDegrees=45
  swing1Motion=Free
  swing2LimitDegrees=45
  swing2Motion=Free
  twistLimitDegrees=45
  twistMotion=Free
 }
 {
  bone1Name=deform_shoulderpad_r
  bone2Name=upperarm_r
  swing1LimitDegrees=45
  swing1Motion=Free
  swing2LimitDegrees=45
  swing2Motion=Free
  twistLimitDegrees=45
  twistMotion=Free
 }
 {
  bone1Name=dyn_shoulderpad_r
  bone2Name=deform_shoulderpad_r
  swing1LimitDegrees=45
  swing1Motion=Free
  swing2LimitDegrees=45
  swing2Motion=Free
  twistLimitDegrees=45
  twistMotion=Free
 }
 {
  bone1Name=dyn_shoulderpad_1_r
  bone2Name=dyn_shoulderpad_r
  swing1LimitDegrees=45
  swing1Motion=Free
  swing2LimitDegrees=45
  swing2Motion=Free
  twistLimitDegrees=45
  twistMotion=Free
 }
 {
  bone1Name=deform_hair_mb_l
  bone2Name=head
  swing1LimitDegrees=45
  swing1Motion=Free
  swing2LimitDegrees=45
  swing2Motion=Free
  twistLimitDegrees=45
  twistMotion=Free
 }
 {bone1Name=dyn_hair_mb_l, bone2Name=deform_hair_mb_l, swing1LimitDegrees=60, swing2LimitDegrees=25, twistLimitDegrees=25}
 {bone1Name=dyn_hair_1_mb_l, bone2Name=dyn_hair_mb_l, swing1LimitDegrees=35, swing2LimitDegrees=15, twistLimitDegrees=10}
 {
  bone1Name=deform_hair_mb_r
  bone2Name=head
  swing1LimitDegrees=45
  swing1Motion=Free
  swing2LimitDegrees=45
  swing2Motion=Free
  twistLimitDegrees=45
  twistMotion=Free
 }
 {bone1Name=dyn_hair_mb_r, bone2Name=deform_hair_mb_r, swing1LimitDegrees=25, swing2LimitDegrees=25, twistLimitDegrees=25}
 {bone1Name=dyn_hair_1_mb_r, bone2Name=dyn_hair_mb_r, swing1LimitDegrees=35, swing2LimitDegrees=15, twistLimitDegrees=10}
 {
  bone1Name=spine_05
  bone2Name=spine_04
  swing1LimitDegrees=45
  swing1Motion=Free
  swing2LimitDegrees=45
  swing2Motion=Free
  twistLimitDegrees=45
  twistMotion=Free
 }
 {
  bone1Name=dyn_hair_bk
  bone2Name=dyn_bandana
  swing1LimitDegrees=5
  swing1Motion=Free
  swing2LimitDegrees=10
  swing2Motion=Free
  twistLimitDegrees=10
  twistMotion=Free
 }
 {
  bone1Name=dyn_hair_b_l
  bone2Name=dyn_hair_mb_l
  swing1LimitDegrees=45
  swing1Motion=Free
  swing2LimitDegrees=45
  swing2Motion=Free
  twistLimitDegrees=45
  twistMotion=Free
 }
 {bone1Name=dyn_hair_1_bk, bone2Name=dyn_hair_1_b_l, swing1LimitDegrees=15, swing2LimitDegrees=15, twistLimitDegrees=15}
 {bone1Name=dyn_hair_1_b_l, bone2Name=dyn_hair_1_mb_l, swing1LimitDegrees=35, swing2LimitDegrees=35, twistLimitDegrees=35}
 {
  bone1Name=dyn_hair_b_r
  bone2Name=dyn_hair_mb_r
  swing1LimitDegrees=15
  swing1Motion=Free
  swing2LimitDegrees=15
  swing2Motion=Free
  twistLimitDegrees=10
  twistMotion=Free
 }
 {
  bone1Name=dyn_hair_1_bk
  bone2Name=dyn_hair_1_b_r
  swing1LimitDegrees=15
  swing1Motion=Free
  swing2LimitDegrees=15
  swing2Motion=Free
  twistLimitDegrees=15
  twistMotion=Free
 }
 {
  bone1Name=dyn_hair_1_b_r
  bone2Name=dyn_hair_1_mb_r
  swing1LimitDegrees=35
  swing1Motion=Free
  swing2LimitDegrees=35
  swing2Motion=Free
  twistLimitDegrees=35
  twistMotion=Free
 }
 {
  bone1Name=dyn_hair_mb_r
  bone2Name=dyn_hair_m_r
  swing1LimitDegrees=15
  swing1Motion=Free
  swing2LimitDegrees=15
  swing2Motion=Free
  twistLimitDegrees=15
  twistMotion=Free
 }
 {
  bone1Name=dyn_hair_1_m_r
  bone2Name=dyn_hair_1_mb_r
  swing1LimitDegrees=15
  swing1Motion=Free
  swing2LimitDegrees=15
  swing2Motion=Free
  twistLimitDegrees=15
  twistMotion=Free
 }
 {
  bone1Name=dyn_hair_1_m_l
  bone2Name=dyn_hair_1_mb_l
  swing1LimitDegrees=45
  swing1Motion=Free
  swing2LimitDegrees=45
  swing2Motion=Free
  twistLimitDegrees=45
  twistMotion=Free
 }
 {bone1Name=dyn_hair_m_r, bone2Name=dyn_hair_1_f_r, swing1LimitDegrees=5, swing2LimitDegrees=5, twistLimitDegrees=5}
]
