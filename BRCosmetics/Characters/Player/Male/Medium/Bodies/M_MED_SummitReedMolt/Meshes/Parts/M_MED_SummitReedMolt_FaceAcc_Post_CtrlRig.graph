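# Post-process Control Rig for the SummitReedMolt male face accessory, correcting hat and hair bones around head, neck and shoulders.
# Hat bones aim at dyn_hair bones, then rotate and scale by weights remapped from hat-to-hair distances.
# Head-versus-spine spherical pose readers and right upper-arm proximity add corrective offsets.
# Side hair bones finally blend toward neck, spine and shoulder bones, with is_in_frontend altering some right-side corrections.
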ControlRigBlueprint /BRCosmetics/Characters/Player/Male/Medium/Bodies/M_MED_SummitReedMolt/Meshes/Parts/M_MED_SummitReedMolt_FaceAcc_Post_CtrlRig
graphs [
 RigVMFunctionLibrary {links [], nodes []}
 RigVMModel {nodes [
   RigUnit_BeginExecution = &RigUnit_BeginExecution()
   n1 = &RigUnit_SphericalPoseReader(
    ActiveRegionScaleFactors=(PositiveWidth=1,NegativeWidth=1,PositiveHeight=2,NegativeHeight=2)
    ActiveRegionSize=0.15
    Debug=(bDrawDebug=false,bDraw2D=False,bDrawLocalAxes=False,DebugScale=25,DebugSegments=20,DebugThickness=0.25)
    DriverAxis=(X=1,Y=0,Z=0)
    DriverItem=(Type=Bone,Name="head")
    FalloffRegionScaleFactors=(PositiveWidth=1,NegativeWidth=1,PositiveHeight=1,NegativeHeight=1)
    FalloffSize=0.8
    FlipHeightScaling=False
    FlipWidthScaling=False
    OptionalParentItem=(Type=Bone,Name="spine_05")
    RotationOffset=(X=-90,Y=0,Z=-40)
    ExecutePin=@RigUnit_BeginExecution.ExecutePin
   )
   > &RigUnit_ModifyTransforms(
    ItemToModify=((Item=(Type=Bone,Name="deform_hair_bk"),Transform=(Rotation=(X=0,Y=0,Z=0,W=1),Translation=(X=0,Y=-2,Z=0),Scale3D=(X=1,Y=1,Z=1))))
    WeightMinimum=0
    Weight=@n1.OutputParam
   )
   > &RigUnit_AimConstraintLocalSpaceOffset(
    Child=(Type=Bone,Name="deform_hat_bk1")
    Filter=(bX=True,bY=True,bZ=True)
    Parents=((Item=(Type=Bone,Name="dyn_hair_1_bk"),Weight=1))
    UpAxis=(X=0,Y=1,Z=0)
    Weight=1
    WorldUp=(Target=(X=0,Y=0,Z=1),Kind=Location,Space=(Type=Bone,Name="deform_hat_bk1"))
   )
   > &RigUnit_AimConstraintLocalSpaceOffset(
    Child=(Type=Bone,Name="deform_hat_b_l")
    Filter=(bX=True,bY=True,bZ=True)
    Parents=((Item=(Type=Bone,Name="dyn_hair_1_b_l"),Weight=1))
    UpAxis=(X=0,Y=1,Z=0)
    Weight=1
    WorldUp=(Target=(X=0,Y=0,Z=1),Kind=Location,Space=(Type=Bone,Name="deform_hat_b_l"))
   )
   > &RigUnit_AimConstraintLocalSpaceOffset(
    Child=(Type=Bone,Name="deform_hat_b_r")
    Filter=(bX=True,bY=True,bZ=True)
    Parents=((Item=(Type=Bone,Name="dyn_hair_1_b_r"),Weight=1),(Item=(Type=Bone,Name="dyn_hair_1_mb_r"),Weight=0))
    UpAxis=(X=0,Y=1,Z=0)
    Weight=1
    WorldUp=(Target=(X=0,Y=0,Z=1),Kind=Location,Space=(Type=Bone,Name="deform_hat_b_r"))
   )
   > &RigUnit_AimConstraintLocalSpaceOffset(
    Child=(Type=Bone,Name="deform_hat_f_l")
    Filter=(bX=True,bY=True,bZ=True)
    Parents=((Item=(Type=Bone,Name="dyn_hair_2_f_l"),Weight=0.6),(Item=(Type=Bone,Name="dyn_hair_1_f_l"),Weight=0.4))
    UpAxis=(X=0,Y=1,Z=0)
    Weight=1
    WorldUp=(Target=(X=0,Y=0,Z=1),Kind=Location,Space=(Type=Bone,Name="deform_hat_f_l"))
   )
   > &RigUnit_AimConstraintLocalSpaceOffset(
    Child=(Type=Bone,Name="deform_hat_f_r")
    Filter=(bX=True,bY=True,bZ=True)
    Parents=((Item=(Type=Bone,Name="dyn_hair_2_f_r"),Weight=0.2),(Item=(Type=Bone,Name="dyn_hair_1_m_r"),Weight=1))
    UpAxis=(X=0,Y=1,Z=0)
    Weight=1
    WorldUp=(Target=(X=0,Y=0,Z=1),Kind=Location,Space=(Type=Bone,Name="deform_hat_f_r"))
   )
   > &RigUnit_AimConstraintLocalSpaceOffset(
    Child=(Type=Bone,Name="deform_hat_m_l")
    Filter=(bX=True,bY=True,bZ=True)
    Parents=((Item=(Type=Bone,Name="dyn_hair_1_m_l"),Weight=1))
    UpAxis=(X=0,Y=1,Z=0)
    Weight=1
    WorldUp=(Target=(X=0,Y=0,Z=1),Kind=Location,Space=(Type=Bone,Name="dyn_hair_m_l"))
   )
   > &RigUnit_AimConstraintLocalSpaceOffset(
    Child=(Type=Bone,Name="deform_hat_m_r")
    Filter=(bX=true,bY=false,bZ=True)
    Parents=((Item=(Type=Bone,Name="dyn_hair_1_m_r"),Weight=0.5),(Item=(Type=Control,Name="dyn_hair_1_mb_r_ctrl"),Weight=0.5))
    UpAxis=(X=0,Y=1,Z=0)
    Weight=1
    WorldUp=(Target=(X=0,Y=0,Z=1),Kind=Location,Space=(Type=Bone,Name="deform_hat_m_r"))
   )
   > &RigUnit_AimConstraintLocalSpaceOffset(
    Child=(Type=Bone,Name="deform_hat_mb_l")
    Filter=(bX=True,bY=True,bZ=True)
    Parents=((Item=(Type=Bone,Name="dyn_hair_1_mb_l"),Weight=1))
    UpAxis=(X=0,Y=1,Z=0)
    Weight=1
    WorldUp=(Target=(X=0,Y=0,Z=1),Kind=Location,Space=(Type=Bone,Name="deform_hat_mb_l"))
   )
   > &RigUnit_AimConstraintLocalSpaceOffset(
    Child=(Type=Bone,Name="deform_hat_mb_r")
    Filter=(bX=true,bY=true,bZ=true)
    Parents=((Item=(Type=Bone,Name="dyn_hair_1_mb_r"),Weight=1),(Item=(Type=Bone,Name="dyn_hair_1_m_r"),Weight=0))
    UpAxis=(X=0,Y=1,Z=0)
    Weight=1
    WorldUp=(Target=(X=0,Y=0,Z=1),Kind=Location,Space=(Type=Bone,Name="deform_hat_mb_r"))
    bMaintainOffset=true
   )
   > &RigUnit_RotationConstraintLocalSpaceOffset(
    AdvancedSettings=(InterpolationType=Average,RotationOrderForFilter=XZY)
    Child=(Type=Bone,Name="deform_hat_bk1")
    Filter=(bX=True,bY=True,bZ=True)
    Parents=((Item=(Type=Bone,Name="deform_hat_b_r"),Weight=0.5),(Item=(Type=Bone,Name="deform_hat_b_l"),Weight=0.5))
    Weight=0.3
   )
   > &RigUnit_ParentConstraint(
    AdvancedSettings=(InterpolationType=Average,RotationOrderForFilter=XZY)
    Child=(Type=Bone,Name="deform_hat_f_l")
    Parents=((Item=(Type=Bone,Name="dyn_hair_f_l"),Weight=1))
    Weight=1
   )
   RigUnit_GetTransform = &RigUnit_GetTransform(Item=(Type=Bone,Name="dyn_hair_1_bk"), bInitial=False)
   RigVMFunction_MathVectorDistance = &RigVMFunction_MathVectorDistance(
    A=@RigUnit_GetTransform.Transform.Translation
    B=&RigUnit_GetTransform(Item=(Type=Bone,Name="deform_hat_bk1"), bInitial=False).Transform.Translation
   )
   Remap = &RigVMFunction_MathDoubleRemap(SourceMaximum=2, SourceMinimum=14, TargetMaximum=1, TargetMinimum=0, bClamp=False, Value=@RigVMFunction_MathVectorDistance.Result)
   > &RigUnit_ModifyTransforms(
    ItemToModify="((Item=(Type=Bone,Name=\"deform_hat_bk1\"),Transform=(Rotation=(X=0,Y=0,Z=0,W=1),Translation=(X=0,Y=0,Z=0),Scale3D=(X=0,Y=1,Z=1))),(Item=(Type=Bone,Name=\"deform_hat_b_l\"),Transform=(Rotation=(X=0,Y=0,Z=0,W=1),Translation=(X=0,Y=0,Z=0),Scale3D=(X=0.5,Y=1,Z=1) ...(truncated)"
    WeightMinimum=0
    Weight=@Remap.Result
   )
   n2 = &RigUnit_GetTransform(Item=(Type=Bone,Name="dyn_hair_1_b_r"), bInitial=False)
   n3 = &RigUnit_GetTransform(Item=(Type=Bone,Name="deform_hat_b_r"), bInitial=False)
   > &RigUnit_ModifyTransforms(
    ItemToModify="((Item=(Type=Bone,Name=\"deform_hat_b_r\"),Transform=(Rotation=(X=0,Y=0,Z=-0.1736,W=0.9848),Translation=(X=0,Y=0,Z=0),Scale3D=(X=0,Y=1,Z=1))),(Item=(Type=Bone,Name=\"deform_hat_bk1\"),Transform=(Rotation=(X=0,Y=0,Z=-0.0872,W=0.9962),Translation=(X=0,Y=0,Z=0),S ...(truncated)"
    WeightMinimum=0
    Weight=&RigVMFunction_MathDoubleRemap(
     SourceMaximum=2
     SourceMinimum=12.6
     TargetMaximum=1
     TargetMinimum=0
     bClamp=False
     Value=&RigVMFunction_MathVectorDistance(A=@n2.Transform.Translation, B=@n3.Transform.Translation).Result
    ).Result
   )
   n4 = &RigUnit_GetTransform(Item=(Type=Bone,Name="dyn_hair_1_b_l"), bInitial=False)
   n5 = &RigUnit_GetTransform(Item=(Type=Bone,Name="deform_hat_b_l"), bInitial=False)
   > &RigUnit_ModifyTransforms(
    ItemToModify="((Item=(Type=Bone,Name=\"deform_hat_b_l\"),Transform=(Rotation=(X=0,Y=0,Z=0,W=1),Translation=(X=0,Y=0,Z=0),Scale3D=(X=0,Y=1,Z=1))),(Item=(Type=Bone,Name=\"deform_hat_bk1\"),Transform=(Rotation=(X=0,Y=0,Z=0,W=1),Translation=(X=0,Y=0,Z=0),Scale3D=(X=0.5,Y=1,Z=1) ...(truncated)"
    WeightMinimum=0
    Weight=&RigVMFunction_MathDoubleRemap(
     SourceMaximum=2
     SourceMinimum=12.6
     TargetMaximum=1
     TargetMinimum=0
     bClamp=False
     Value=&RigVMFunction_MathVectorDistance(A=@n4.Transform.Translation, B=@n5.Transform.Translation).Result
    ).Result
   )
   n6 = &RigUnit_GetTransform(Item=(Type=Bone,Name="dyn_hair_1_mb_r"), bInitial=False)
   n7 = &RigUnit_GetTransform(Item=(Type=Bone,Name="deform_hat_mb_r"), bInitial=False)
   > &RigUnit_ModifyTransforms(
    ItemToModify="((Item=(Type=Bone,Name=\"deform_hat_mb_r\"),Transform=(Rotation=(X=0.0449,Y=0.2549,Z=-0.1677,W=0.9513),Translation=(X=0,Y=0,Z=0),Scale3D=(X=0,Y=1,Z=1))),(Item=(Type=Bone,Name=\"deform_hat_m_r\"),Transform=(Rotation=(X=0.0227,Y=0.1285,Z=-0.1722,W=0.9764),Transl ...(truncated)"
    WeightMinimum=0
    Weight=&RigVMFunction_MathDoubleRemap(
     SourceMaximum=2
     SourceMinimum=11.6
     TargetMaximum=1
     TargetMinimum=0
     bClamp=False
     Value=&RigVMFunction_MathVectorDistance(A=@n6.Transform.Translation, B=@n7.Transform.Translation).Result
    ).Result
   )
   n8 = &RigUnit_GetTransform(Item=(Type=Bone,Name="dyn_hair_1_mb_l"), bInitial=False)
   n9 = &RigUnit_GetTransform(Item=(Type=Bone,Name="deform_hat_mb_l"), bInitial=False)
   > &RigUnit_ModifyTransforms(
    ItemToModify="((Item=(Type=Bone,Name=\"deform_hat_mb_l\"),Transform=(Rotation=(X=-0.0449,Y=-0.2549,Z=-0.1677,W=0.9513),Translation=(X=0,Y=0,Z=0),Scale3D=(X=0,Y=1,Z=1))),(Item=(Type=Bone,Name=\"deform_hat_m_l\"),Transform=(Rotation=(X=-0.0227,Y=-0.1285,Z=-0.1722,W=0.9764),Tr ...(truncated)"
    WeightMinimum=0
    Weight=&RigVMFunction_MathDoubleRemap(
     SourceMaximum=2
     SourceMinimum=11.6
     TargetMaximum=1
     TargetMinimum=0
     bClamp=False
     Value=&RigVMFunction_MathVectorDistance(A=@n8.Transform.Translation, B=@n9.Transform.Translation).Result
    ).Result
   )
   n10 = &RigUnit_GetTransform(Item=(Type=Bone,Name="dyn_hair_1_m_r"), bInitial=False)
   n11 = &RigUnit_GetTransform(Item=(Type=Bone,Name="deform_hat_m_r"), bInitial=False)
   > &RigUnit_ModifyTransforms(
    ItemToModify="((Item=(Type=Bone,Name=\"deform_hat_m_r\"),Transform=(Rotation=(X=0,Y=0,Z=-0.1736,W=0.9848),Translation=(X=0,Y=0,Z=0),Scale3D=(X=0,Y=1,Z=1))),(Item=(Type=Bone,Name=\"deform_hat_f_r\"),Transform=(Rotation=(X=0,Y=0,Z=-0.0872,W=0.9962),Translation=(X=0,Y=0,Z=0),S ...(truncated)"
    WeightMinimum=0
    Weight=&RigVMFunction_MathDoubleRemap(
     SourceMaximum=2
     SourceMinimum=10.6
     TargetMaximum=1
     TargetMinimum=0
     bClamp=False
     Value=&RigVMFunction_MathVectorDistance(A=@n10.Transform.Translation, B=@n11.Transform.Translation).Result
    ).Result
   )
   n12 = &RigUnit_GetTransform(Item=(Type=Bone,Name="dyn_hair_1_m_l"), bInitial=False)
   n13 = &RigUnit_GetTransform(Item=(Type=Bone,Name="deform_hat_m_l"), bInitial=False)
   > &RigUnit_ModifyTransforms(
    ItemToModify="((Item=(Type=Bone,Name=\"deform_hat_m_l\"),Transform=(Rotation=(X=0,Y=0,Z=-0.1736,W=0.9848),Translation=(X=0,Y=0,Z=0),Scale3D=(X=0,Y=1,Z=1))),(Item=(Type=Bone,Name=\"deform_hat_f_l\"),Transform=(Rotation=(X=0,Y=0,Z=-0.0872,W=0.9962),Translation=(X=0,Y=0,Z=0),S ...(truncated)"
    WeightMinimum=0
    Weight=&RigVMFunction_MathDoubleRemap(
     SourceMaximum=2
     SourceMinimum=10.6
     TargetMaximum=1
     TargetMinimum=0
     bClamp=False
     Value=&RigVMFunction_MathVectorDistance(A=@n12.Transform.Translation, B=@n13.Transform.Translation).Result
    ).Result
   )
   n14 = &RigUnit_GetTransform(Item=(Type=Bone,Name="dyn_hair_3_f_r"), bInitial=False)
   n15 = &RigUnit_GetTransform(Item=(Type=Bone,Name="deform_hat_f_r"), bInitial=False)
   > &RigUnit_ModifyTransforms(
    ItemToModify=((Item=(Type=Bone,Name="deform_hat_f_r"),Transform=(Rotation=(X=0,Y=0,Z=0,W=1),Translation=(X=0,Y=0,Z=0),Scale3D=(X=0,Y=1,Z=1))))
    WeightMinimum=0
    Weight=&RigVMFunction_MathDoubleRemap(
     SourceMaximum=0
     SourceMinimum=11.6
     TargetMaximum=1
     TargetMinimum=0
     bClamp=False
     Value=&RigVMFunction_MathVectorDistance(A=@n14.Transform.Translation, B=@n15.Transform.Translation).Result
    ).Result
   )
   n16 = &RigUnit_GetTransform(Item=(Type=Bone,Name="dyn_hair_3_f_l"), bInitial=False)
   n17 = &RigUnit_GetTransform(Item=(Type=Bone,Name="deform_hat_f_l"), bInitial=False)
   > &RigUnit_ModifyTransforms(
    ItemToModify=((Item=(Type=Bone,Name="deform_hat_f_l"),Transform=(Rotation=(X=0,Y=0,Z=0,W=1),Translation=(X=0,Y=0,Z=0),Scale3D=(X=0,Y=1,Z=1))))
    WeightMinimum=0
    Weight=&RigVMFunction_MathDoubleRemap(
     SourceMaximum=0
     SourceMinimum=11.6
     TargetMaximum=1
     TargetMinimum=0
     bClamp=False
     Value=&RigVMFunction_MathVectorDistance(A=@n16.Transform.Translation, B=@n17.Transform.Translation).Result
    ).Result
   )
   > n18 = &RigUnit_SphericalPoseReader(
    ActiveRegionScaleFactors=(PositiveWidth=1,NegativeWidth=2,PositiveHeight=2,NegativeHeight=1)
    ActiveRegionSize=0.1
    Debug=(bDrawDebug=true,bDraw2D=False,bDrawLocalAxes=False,DebugScale=25,DebugSegments=20,DebugThickness=0.25)
    DriverAxis=(X=0,Y=1,Z=0)
    DriverItem=(Type=Bone,Name="head")
    FalloffRegionScaleFactors=(PositiveWidth=1,NegativeWidth=1,PositiveHeight=1,NegativeHeight=0.5)
    FalloffSize=0.5
    FlipHeightScaling=False
    FlipWidthScaling=False
    OptionalParentItem=(Type=Bone,Name="spine_05")
    RotationOffset=(X=-10,Y=0,Z=0)
   )
   RigVMFunction_MathBoolToFloat = &RigVMFunction_MathBoolToFloat(Value=RigVMVariableNode() {variable=is_in_frontend}.Value)
   AlphaInterp = &RigVMFunction_AlphaInterp(
    InterpSpeedDecreasing=10
    InterpSpeedIncreasing=10
    OutRange=(Min=1,Max=0)
    Scale=1
    bClampResult=true
    bInterpResult=False
    bMapRange=true
    Value=@RigVMFunction_MathBoolToFloat.Result
   )
   Multiply = &RigVMFunction_MathDoubleMul(B=@AlphaInterp.Result, A=@n18.OutputParam)
   > &RigUnit_ModifyTransforms(
    ItemToModify="((Item=(Type=Bone,Name=\"deform_hat_f_r\"),Transform=(Rotation=(X=0,Y=0,Z=-0.1736,W=0.9848),Translation=(X=0,Y=0,Z=0),Scale3D=(X=0.7,Y=1,Z=1))),(Item=(Type=Bone,Name=\"deform_hair_b_r\"),Transform=(Rotation=(X=0,Y=0,Z=0,W=1),Translation=(X=0,Y=0,Z=0),Scale3D=( ...(truncated)"
    WeightMinimum=0
    Weight=@Multiply.Result
   )
   > n19 = &RigUnit_SphericalPoseReader(
    ActiveRegionScaleFactors=(PositiveWidth=1,NegativeWidth=2,PositiveHeight=2,NegativeHeight=1)
    ActiveRegionSize=0.1
    Debug=(bDrawDebug=false,bDraw2D=False,bDrawLocalAxes=False,DebugScale=25,DebugSegments=20,DebugThickness=0.25)
    DriverAxis=(X=0,Y=1,Z=0)
    DriverItem=(Type=Bone,Name="head")
    FalloffRegionScaleFactors=(PositiveWidth=1,NegativeWidth=1,PositiveHeight=1,NegativeHeight=0.5)
    FalloffSize=0.5
    FlipHeightScaling=False
    FlipWidthScaling=False
    OptionalParentItem=(Type=Bone,Name="spine_05")
    RotationOffset=(X=180,Y=0,Z=0)
   )
   > &RigUnit_ModifyTransforms(
    ItemToModify="((Item=(Type=Bone,Name=\"deform_hat_f_l\"),Transform=(Rotation=(X=0,Y=0,Z=-0.1736,W=0.9848),Translation=(X=0,Y=0,Z=0),Scale3D=(X=0.7,Y=1,Z=1))),(Item=(Type=Bone,Name=\"deform_hair_b_r\"),Transform=(Rotation=(X=0,Y=0,Z=0,W=1),Translation=(X=0,Y=0,Z=0),Scale3D=( ...(truncated)"
    WeightMinimum=0
    Weight=@n19.OutputParam
   )
   n20 = &RigVMFunction_MathDoubleRemap(
    SourceMaximum=20
    SourceMinimum=30
    TargetMaximum=1
    TargetMinimum=0
    Value=&RigVMFunction_MathVectorDistance(
     B=&RigUnit_GetTransform(Item=(Type=Bone,Name="upperarm_r"), bInitial=False).Transform.Translation
     A=&RigUnit_GetTransform(Item=(Type=Control,Name="head_ctrl"), bInitial=False).Transform.Translation
    ).Result
   )
   > &RigUnit_ModifyTransforms(
    ItemToModify="((Item=(Type=Bone,Name=\"deform_hat_b_r\"),Transform=(Rotation=(X=0,Y=0,Z=-0.1736,W=0.9848),Translation=(X=0,Y=0,Z=0),Scale3D=(X=1,Y=1,Z=1))),(Item=(Type=Bone,Name=\"deform_hat_mb_r\"),Transform=(Rotation=(X=0,Y=0,Z=-0.2588,W=0.9659),Translation=(X=0,Y=0,Z=0), ...(truncated)"
    WeightMinimum=0
    Weight=@n20.Result
   )
   > &RigUnit_ParentConstraint(
    AdvancedSettings=(RotationOrderForFilter=XZY)
    Child=(Type=Bone,Name="dyn_hair_1_m_r")
    Filter=(TranslationFilter=(bX=True,bY=True,bZ=True),RotationFilter=(bX=true,bY=true,bZ=true),ScaleFilter=(bX=True,bY=True,bZ=True))
    Parents=((Item=(Type=Bone,Name="head"),Weight=0.0256),(Item=(Type=Bone,Name="clavicle_r"),Weight=0.0306),(Item=(Type=Bone,Name="spine_05"),Weight=0.2251),(Item=(Type=Bone,Name="neck_02"),Weight=0.3592),(Item=(Type=Bone,Name="neck_01"),Weight=0.3595))
    Weight=0.7
   )
   > &RigUnit_ParentConstraint(
    AdvancedSettings=(RotationOrderForFilter=XZY)
    Child=(Type=Bone,Name="dyn_hair_1_mb_r")
    Filter=(TranslationFilter=(bX=True,bY=True,bZ=True),RotationFilter=(bX=true,bY=true,bZ=true),ScaleFilter=(bX=True,bY=True,bZ=True))
    Parents="((Item=(Type=Bone,Name=\"head\"),Weight=0.0013),(Item=(Type=Bone,Name=\"spine_03\"),Weight=0.002),(Item=(Type=Bone,Name=\"upperarm_twist_01_r\"),Weight=0.003),(Item=(Type=Bone,Name=\"deform_upperarm_up_r\"),Weight=0.0221),(Item=(Type=Bone,Name=\"neck_02\"),Weight=0. ...(truncated)"
    Weight=0.8
   )
   > &RigUnit_ParentConstraint(
    AdvancedSettings=(RotationOrderForFilter=XZY)
    Child=(Type=Bone,Name="dyn_hair_1_m_l")
    Filter=(TranslationFilter=(bX=True,bY=True,bZ=True),RotationFilter=(bX=true,bY=true,bZ=true),ScaleFilter=(bX=True,bY=True,bZ=True))
    Parents=((Item=(Type=Bone,Name="head"),Weight=0.0005),(Item=(Type=Bone,Name="clavicle_l"),Weight=0.0568),(Item=(Type=Bone,Name="neck_02"),Weight=0.2851),(Item=(Type=Bone,Name="neck_01"),Weight=0.327),(Item=(Type=Bone,Name="spine_05"),Weight=0.3306))
    Weight=0.5
   )
   > &RigUnit_ParentConstraint(
    AdvancedSettings=(RotationOrderForFilter=XZY)
    Child=(Type=Bone,Name="dyn_hair_1_mb_l")
    Filter=(TranslationFilter=(bX=True,bY=True,bZ=True),RotationFilter=(bX=true,bY=true,bZ=true),ScaleFilter=(bX=True,bY=True,bZ=True))
    Parents="((Item=(Type=Bone,Name=\"deform_upperarm_up_l\"),Weight=0.0003),(Item=(Type=Bone,Name=\"head\"),Weight=0.0071),(Item=(Type=Bone,Name=\"clavicle_l\"),Weight=0.055),(Item=(Type=Bone,Name=\"neck_02\"),Weight=0.2572),(Item=(Type=Bone,Name=\"neck_01\"),Weight=0.3261),(It ...(truncated)"
    Weight=0.8
   )
   > &RigUnit_ModifyTransforms(
    ItemToModify="((Item=(Type=Bone,Name=\"dyn_hair_1_mb_r\"),Transform=(Rotation=(X=0.2588,Y=0,Z=0,W=0.9659),Translation=(X=0,Y=0,Z=0),Scale3D=(X=0.7,Y=0.5,Z=1))),(Item=(Type=Bone,Name=\"dyn_hair_1_mb_l\"),Transform=(Rotation=(X=0.2588,Y=0,Z=0,W=0.9659),Translation=(X=0,Y=0,Z= ...(truncated)"
    Weight=1
    WeightMinimum=0
   )
   > &RigUnit_ModifyTransforms(
    ItemToModify=((Item=(Type=Bone,Name="dyn_hair_m_r"),Transform=(Rotation=(X=-0.0872,Y=0,Z=0,W=0.9962),Translation=(X=0,Y=0,Z=0),Scale3D=(X=1,Y=1,Z=1))))
    WeightMinimum=0
    Weight=&RigVMFunction_MathBoolToFloat(Value=RigVMVariableNode() {variable=is_in_frontend}.Value).Result
   )
   &RigUnit_AimConstraintLocalSpaceOffset(
    Child=(Type=Bone,Name="deform_hat_f_r")
    ExecutePin=()
    Filter=(bX=True,bY=True,bZ=True)
    Parents=((Item=(Type=Bone,Name="dyn_hair_3_f_r"),Weight=0.7),(Item=(Type=Bone,Name="dyn_hair_1_m_r"),Weight=0.3))
    UpAxis=(X=0,Y=1,Z=0)
    Weight=1
    WorldUp=(Target=(X=0,Y=0,Z=1),Kind=Location,Space=(Type=Bone,Name="deform_hat_f_r"))
   )
   &RigVMFunction_AnimEvalRichCurve(
    Curve=(EditorCurveData=(Keys=((),(Time=0.275,Value=0.2571),(Time=0.5649,Value=0.4215),(Time=1,Value=0.8,ArriveTangent=0.8754,LeaveTangent=0.8754))))
    SourceMinimum=0
    TargetMinimum=0
    Value=@n20.Result
   )
  ]}
]
hierarchy [
 root {loc=<0 0 0>, rot=<0 0 0 1>}
 pelvis {parent=root, loc=<0 -0.7071 93.1191>, rot=<0.0127 0.707 -0.0127 -0.707>}
 spine_01 {parent=pelvis, loc=<6.7494 -0.6388 0>, rot=<0 0 0.034 -0.9994>}
 spine_02 {parent=spine_01, loc=<7.2518 -0.035 0>, rot=<0 0 0.0238 -0.9997>}
 spine_03 {parent=spine_02, loc=<6.0104 0 0>, rot=<0 0 0.0149 -0.9999>}
 spine_04 {parent=spine_03, loc=<8.1024 0 0>, rot=<0 0 -0.0132 -0.9999>}
 spine_05 {parent=spine_04, loc=<10.0146 0 0>, rot=<0 0 -0.0015 -1>}
 neck_01 {parent=spine_05, loc=<11.3138 0.1529 0>, rot=<0 0 -0.2003 -0.9797>}
 neck_02 {parent=neck_01, loc=<4.5627 0 0>, rot=<0 0 0.0504 -0.9987>}
 head {parent=neck_02, loc=<5.3914 0 0>, rot=<0 0 0.115 -0.9934>}
 deform_hair_bk {parent=head, loc=<2.4762 -6.7734 0>, rot=<0.1716 -0.9852 0 0>}
 dyn_hair_bk {parent=deform_hair_bk, loc=<0 0 0>, rot=<0 0 0 1>}
 dyn_hair_1_bk {parent=dyn_hair_bk, loc=<8.7869 0 0>, rot=<0 0 0 1>}
 deform_hair_b_l {parent=head, loc=<2.01 -5.8039 -4.0624>, rot=<0.1377 -0.9526 -0.2696 -0.031>}
 dyn_hair_b_l {parent=deform_hair_b_l, loc=<0 0 0>, rot=<0 0 0 1>}
 dyn_hair_1_b_l {parent=dyn_hair_b_l, loc=<7.8326 0 0>, rot=<0 0 0 1>}
 deform_hair_m_l {parent=head, loc=<1.8924 -1.0847 -6.2281>, rot=<-0.064 0.7029 0.7052 0.0673>}
 dyn_hair_m_l {parent=deform_hair_m_l, loc=<0 0 0>, rot=<0 0 0 1>}
 dyn_hair_1_m_l {parent=dyn_hair_m_l, loc=<6.7334 0 0>, rot=<0 0 0 1>}
 deform_hair_f_l {parent=head, loc=<3.6691 3.9712 -6.4624>, rot=<-0.0587 0.6155 0.7858 -0.0176>}
 dyn_hair_f_l {parent=deform_hair_f_l, loc=<0 0 0>, rot=<0 0 0 1>}
 dyn_hair_1_f_l {parent=dyn_hair_f_l, loc=<5.4397 0 0>, rot=<0 0 0 1>}
 dyn_hair_3_f_l {parent=dyn_hair_1_f_l, loc=<4.8384 -0.5351 -0.9281>, rot=<-0.029 -0.0381 0.206 -0.9774>}
 dyn_hair_2_f_l {parent=dyn_hair_1_f_l, loc=<5.0126 -0.1226 1.3997>, rot=<-0.007 0.0786 0.0889 -0.9929>}
 deform_hair_f_r {parent=head, loc=<3.6691 3.9712 6.4624>, rot=<0.0587 -0.6155 0.7858 -0.0176>}
 dyn_hair_f_r {parent=deform_hair_f_r, loc=<0 0 0>, rot=<0 0 0 -1>}
 dyn_hair_1_f_r {parent=dyn_hair_f_r, loc=<5.4395 0 0>, rot=<0 0 0 -1>}
 dyn_hair_3_f_r {parent=dyn_hair_1_f_r, loc=<4.8386 -0.5351 0.9281>, rot=<0.029 0.0381 0.206 -0.9774>}
 dyn_hair_2_f_r {parent=dyn_hair_1_f_r, loc=<5.0132 -0.1226 -1.3998>, rot=<0.007 -0.0786 0.0889 -0.9929>}
 deform_hair_b_r {parent=head, loc=<2.0098 -5.8039 4.0624>, rot=<0.1377 -0.9526 0.2696 0.031>}
 dyn_hair_b_r {parent=deform_hair_b_r, loc=<0 0 0>, rot=<0 0 0 1>}
 dyn_hair_1_b_r {parent=dyn_hair_b_r, loc=<7.8326 0 0>, rot=<0 0 0 1>}
 deform_bandana {parent=head, loc=<3.7397 -9.0352 0>, rot=<0.1716 -0.9852 0 0>}
 dyn_bandana {parent=deform_bandana, loc=<0 0 0>, rot=<0 0 0 1>}
 dyn_bandana_l {parent=dyn_bandana, loc=<0.2918 0.1528 0.8925>, rot=<0.0078 0.2101 -0.0424 -0.9767>}
 dyn_bandana_r {parent=dyn_bandana, loc=<0.463 0.1426 -1.083>, rot=<-0.1904 -0.2848 -0.0642 -0.9373>}
 FACIAL_C_FacialRoot {parent=head, loc=<2.1524 -0.746 -0.0002>, rot=<0.4879 -0.5118 -0.4879 -0.5118>}
 FACIAL_L_EyelidUpperA {parent=FACIAL_C_FacialRoot, loc=<2.6183 -1.7238 8.2542>, rot=<-0.1478 -0.0001 0 -0.989>}
 FACIAL_R_EyelidUpperA {parent=FACIAL_C_FacialRoot, loc=<-2.6195 -1.7204 8.2544>, rot=<-0.1479 -0.0001 0 -0.989>}
 FACIAL_L_Eye {parent=FACIAL_C_FacialRoot, loc=<2.6153 -1.707 8.2708>, rot=<0 -0.0001 0 -1>}
 FACIAL_R_Eye {parent=FACIAL_C_FacialRoot, loc=<-2.6165 -1.7051 8.2712>, rot=<0 -0.0001 0 -1>}
 FACIAL_L_EyelidLowerA {parent=FACIAL_C_FacialRoot, loc=<2.6183 -1.7238 8.2542>, rot=<0.284 -0.0001 -0.0001 -0.9588>}
 FACIAL_R_EyelidLowerA {parent=FACIAL_C_FacialRoot, loc=<-2.6195 -1.7204 8.2543>, rot=<0.284 -0.0001 -0.0001 -0.9588>}
 FACIAL_L_EyeCornerInner {parent=FACIAL_C_FacialRoot, loc=<1.6728 -1.3164 8.9079>, rot=<0 -0.0001 0 -1>}
 FACIAL_R_EyeCornerInner {parent=FACIAL_C_FacialRoot, loc=<-1.6633 -1.2883 8.8902>, rot=<0 -0.0001 0 -1>}
 FACIAL_L_EyeCornerOuter {parent=FACIAL_C_FacialRoot, loc=<3.8654 -1.5187 8.4932>, rot=<0 -0.3827 0 -0.9239>}
 FACIAL_R_EyeCornerOuter {parent=FACIAL_C_FacialRoot, loc=<-3.8723 -1.519 8.5073>, rot=<-0.0001 0.3825 0 -0.9239>}
 FACIAL_L_EyesackLower {parent=FACIAL_C_FacialRoot, loc=<2.8753 -0.2779 8.5246>, rot=<-0.0223 -0.1398 -0.0223 -0.9897>}
 FACIAL_R_EyesackLower {parent=FACIAL_C_FacialRoot, loc=<-2.8755 -0.2741 8.5191>, rot=<-0.0223 0.1396 0.0222 -0.9897>}
 FACIAL_L_CheekInner {parent=FACIAL_C_FacialRoot, loc=<1.8168 0.3531 8.6628>, rot=<-0.1233 -0.0881 -0.1971 -0.9686>}
 FACIAL_R_CheekInner {parent=FACIAL_C_FacialRoot, loc=<-1.8167 0.36 8.6598>, rot=<-0.1232 0.0879 0.1971 -0.9686>}
 FACIAL_L_CheekOuter {parent=FACIAL_C_FacialRoot, loc=<3.4545 0.4204 7.0737>, rot=<0.0552 -0.377 0.1003 -0.9191>}
 FACIAL_R_CheekOuter {parent=FACIAL_C_FacialRoot, loc=<-3.4549 0.4311 7.0723>, rot=<0.0551 0.3769 -0.1004 -0.9192>}
 FACIAL_L_NasolabialBulge {parent=FACIAL_C_FacialRoot, loc=<3.322 3.4707 6.9224>, rot=<0.0579 -0.3774 0.0238 -0.9239>}
 FACIAL_R_NasolabialBulge {parent=FACIAL_C_FacialRoot, loc=<-3.3243 3.4738 6.9209>, rot=<0.0579 0.3773 -0.0238 -0.924>}
 FACIAL_L_NasolabialFurrow {parent=FACIAL_C_FacialRoot, loc=<2.6659 2.3026 9.2946>, rot=<-0.0333 -0.1259 -0.3637 -0.9224>}
 FACIAL_R_NasolabialFurrow {parent=FACIAL_C_FacialRoot, loc=<-2.6672 2.3056 9.2942>, rot=<-0.0333 0.1257 0.3636 -0.9224>}
 FACIAL_L_CheekLower {parent=FACIAL_C_FacialRoot, loc=<3.8957 3.1852 6.2023>, rot=<0.0346 -0.4269 0.0577 -0.9018>}
 FACIAL_R_CheekLower {parent=FACIAL_C_FacialRoot, loc=<-3.8996 3.1873 6.1992>, rot=<0.0345 0.4268 -0.0578 -0.9018>}
 FACIAL_C_Nose {parent=FACIAL_C_FacialRoot, loc=<-0.0018 -0.2254 10.1702>, rot=<-0.1304 -0.0001 0 -0.9915>}
 FACIAL_C_NoseLower {parent=FACIAL_C_Nose, loc=<0.0011 2.7766 -0.1111>, rot=<0.1305 0 0 -0.9914>}
 FACIAL_L_Nostril {parent=FACIAL_C_Nose, loc=<1.3529 2.2703 0.0558>, rot=<0.0173 0.1305 0.0023 -0.9913>}
 FACIAL_R_Nostril {parent=FACIAL_C_Nose, loc=<-1.3511 2.2425 0.0337>, rot=<0.0173 -0.1305 -0.0023 -0.9913>}
 FACIAL_C_NoseTip {parent=FACIAL_C_Nose, loc=<-0.0012 1.8115 0.683>, rot=<0.0872 0 0 -0.9962>}
 FACIAL_C_MouthUpper {parent=FACIAL_C_FacialRoot, loc=<0.0002 3.898 7.9619>, rot=<0 -0.0001 0 -1>}
 FACIAL_C_LipUpper {parent=FACIAL_C_MouthUpper, loc=<-0.0013 -0.2823 2.6775>, rot=<-0.0872 0 0 -0.9962>}
 FACIAL_L_LipUpper {parent=FACIAL_C_MouthUpper, loc=<1.408 -0.1222 2.3057>, rot=<-0.0778 -0.1601 -0.0379 -0.9833>}
 FACIAL_R_LipUpper {parent=FACIAL_C_MouthUpper, loc=<-1.4097 -0.1187 2.3085>, rot=<-0.0778 0.1601 0.0379 -0.9833>}
 FACIAL_L_LipUpperOuter {parent=FACIAL_C_MouthUpper, loc=<1.8752 0.0874 1.8209>, rot=<-0.0984 -0.2293 -0.0983 -0.9634>}
 FACIAL_R_LipUpperOuter {parent=FACIAL_C_MouthUpper, loc=<-1.8738 0.1004 1.8266>, rot=<-0.0984 0.2293 0.0983 -0.9634>}
 FACIAL_L_LipCorner {parent=FACIAL_C_MouthUpper, loc=<2.0073 0.4382 1.2891>, rot=<0 -0.2078 0 -0.9782>}
 FACIAL_R_LipCorner {parent=FACIAL_C_MouthUpper, loc=<-2.0648 0.432 1.3194>, rot=<0 0.2078 0 -0.9782>}
 FACIAL_L_Masseter {parent=FACIAL_C_FacialRoot, loc=<5.6144 1.1337 4.7461>, rot=<0.0693 0.1128 0.0079 -0.9912>}
 FACIAL_R_Masseter {parent=FACIAL_C_FacialRoot, loc=<-5.6172 1.1385 4.7401>, rot=<0.0693 -0.113 -0.0079 -0.9911>}
 FACIAL_C_UnderChin {parent=FACIAL_C_FacialRoot, loc=<-0.0005 8.0759 5.4207>, rot=<0.4998 -0.0001 -0.0001 -0.8661>}
 FACIAL_L_UnderChin {parent=FACIAL_C_FacialRoot, loc=<3.0851 6.8231 3.9627>, rot=<0.4025 -0.3549 0.0822 -0.8398>}
 FACIAL_R_UnderChin {parent=FACIAL_C_FacialRoot, loc=<-3.0977 6.8322 3.9572>, rot=<0.4025 0.3548 -0.0824 -0.8398>}
 FACIAL_C_TeethUpper {parent=FACIAL_C_FacialRoot, loc=<-0.0006 4.1655 9.2413>, rot=<0 -0.0001 0 -1>}
 FACIAL_C_LowerLipRotation {parent=FACIAL_C_FacialRoot, loc=<-0.0046 0.8638 2.7105>, rot=<0.342 -0.0001 -0.0001 -0.9397>}
 FACIAL_C_MouthLower {parent=FACIAL_C_LowerLipRotation, loc=<0.0041 -1.0513 5.9732>, rot=<-0.342 0 0 -0.9397>}
 FACIAL_C_LipLower {parent=FACIAL_C_MouthLower, loc=<0.0024 1.6381 2.2679>, rot=<0.0872 0 0 -0.9962>}
 FACIAL_L_LipLower {parent=FACIAL_C_MouthLower, loc=<0.9466 1.4812 1.997>, rot=<0.0927 -0.1294 0.0272 -0.9869>}
 FACIAL_R_LipLower {parent=FACIAL_C_MouthLower, loc=<-0.9472 1.4843 2.0054>, rot=<0.0927 0.1294 -0.0272 -0.9869>}
 FACIAL_L_LipLowerOuter {parent=FACIAL_C_MouthLower, loc=<1.635 0.96 1.6771>, rot=<0.0913 -0.2321 0.0318 -0.9679>}
 FACIAL_R_LipLowerOuter {parent=FACIAL_C_MouthLower, loc=<-1.6391 0.9611 1.6835>, rot=<0.0913 0.2321 -0.0318 -0.9679>}
 FACIAL_C_TeethLower {parent=FACIAL_C_LowerLipRotation, loc=<0.003 -1.6302 7.0465>, rot=<-0.342 0 0 -0.9397>}
 FACIAL_C_Tongue1 {parent=FACIAL_C_TeethLower, loc=<0.0024 -0.0759 -5.2082>, rot=<-0.2588 0 0 -0.9659>}
 FACIAL_C_Tongue2 {parent=FACIAL_C_Tongue1, loc=<-0.0003 0.4736 1.4654>, rot=<0.2588 0 0 -0.9659>}
 FACIAL_C_Tongue3 {parent=FACIAL_C_Tongue2, loc=<-0.0005 0.2021 1.4104>, rot=<0.1305 0 0 -0.9914>}
 FACIAL_C_Tongue4 {parent=FACIAL_C_Tongue3, loc=<0.0002 -0.1684 1.151>, rot=<0.0436 0 0 -0.999>}
 FACIAL_C_TongueUpper3 {parent=FACIAL_C_Tongue3, loc=<-0.0004 -0.3898 -0.1005>, rot=<0 0 0 1>}
 FACIAL_C_TongueLower3 {parent=FACIAL_C_Tongue3, loc=<0.0003 0.7854 0.1203>, rot=<0 0 0 1>}
 FACIAL_L_TongueSide3 {parent=FACIAL_C_Tongue3, loc=<1.545 0.0512 0.2077>, rot=<0 0 0 1>}
 FACIAL_R_TongueSide3 {parent=FACIAL_C_Tongue3, loc=<-1.5452 0.0523 0.2091>, rot=<0 0 0 1>}
 FACIAL_C_TongueUpper2 {parent=FACIAL_C_Tongue2, loc=<-0.0005 -0.533 0.1395>, rot=<0 0 0 1>}
 FACIAL_L_TongueSide2 {parent=FACIAL_C_Tongue2, loc=<1.8262 0.1897 0.524>, rot=<0 0 0 1>}
 FACIAL_R_TongueSide2 {parent=FACIAL_C_Tongue2, loc=<-1.8257 0.1903 0.5199>, rot=<0 0 0 1>}
 FACIAL_C_Jaw {parent=FACIAL_C_FacialRoot, loc=<-0.0046 0.8638 2.7105>, rot=<0.342 -0.0001 -0.0001 -0.9397>}
 FACIAL_C_Jawline {parent=FACIAL_C_Jaw, loc=<0.0039 2.3926 8.5301>, rot=<0.2586 0 0 -0.966>}
 FACIAL_L_ChinSide {parent=FACIAL_C_Jaw, loc=<2.4594 0.5713 7.7925>, rot=<-0.2208 -0.3738 0.0499 -0.8995>}
 FACIAL_R_ChinSide {parent=FACIAL_C_Jaw, loc=<-2.4464 0.569 7.7934>, rot=<-0.2208 0.3738 -0.0499 -0.8995>}
 FACIAL_C_Chin {parent=FACIAL_C_Jaw, loc=<0.004 0.959 8.3844>, rot=<-0.2591 0 0 -0.9659>}
 FACIAL_L_Ear {parent=FACIAL_C_FacialRoot, loc=<6.0421 -0.1628 1.3217>, rot=<-0.0263 -0.0008 -0.0262 -0.9993>}
 FACIAL_R_Ear {parent=FACIAL_C_FacialRoot, loc=<-6.0488 -0.1666 1.3161>, rot=<-0.0263 0.0006 0.0261 -0.9993>}
 FACIAL_L_Jawline {parent=FACIAL_C_FacialRoot, loc=<4.0755 4.9701 5.146>, rot=<0.2093 0.1577 -0.0547 -0.9635>}
 FACIAL_R_Jawline {parent=FACIAL_C_FacialRoot, loc=<-4.0795 4.972 5.141>, rot=<0.2093 -0.1578 0.0546 -0.9635>}
 FACIAL_C_NoseBridge {parent=FACIAL_C_FacialRoot, loc=<-0.0005 -2.391 10.5427>, rot=<0.0438 -0.0001 0 -0.999>}
 FACIAL_L_EyelidUpperB {parent=FACIAL_C_FacialRoot, loc=<2.6183 -1.7238 8.2542>, rot=<-0.225 -0.0001 0 -0.9744>}
 FACIAL_L_EyelidLowerB {parent=FACIAL_C_FacialRoot, loc=<2.6183 -1.7238 8.2542>, rot=<0.3173 -0.0001 -0.0001 -0.9483>}
 FACIAL_R_EyelidLowerB {parent=FACIAL_C_FacialRoot, loc=<-2.6195 -1.7204 8.2544>, rot=<0.3255 -0.0001 -0.0001 -0.9455>}
 FACIAL_R_EyelidUpperB {parent=FACIAL_C_FacialRoot, loc=<-2.6195 -1.7204 8.2544>, rot=<-0.225 -0.0001 0 -0.9744>}
 FACIAL_L_ForeheadOut {parent=FACIAL_C_FacialRoot, loc=<3.8728 -3.5851 8.4478>, rot=<-0.1214 -0.2557 -0.0792 -0.9558>}
 FACIAL_L_ForeheadMid {parent=FACIAL_C_FacialRoot, loc=<2.7491 -3.673 9.2095>, rot=<-0.071 -0.1757 -0.0206 -0.9817>}
 FACIAL_R_ForeheadOut {parent=FACIAL_C_FacialRoot, loc=<-3.8742 -3.5864 8.449>, rot=<-0.1214 0.2555 0.0791 -0.9559>}
 FACIAL_R_ForeheadMid {parent=FACIAL_C_FacialRoot, loc=<-2.7516 -3.674 9.2092>, rot=<-0.071 0.1756 0.0206 -0.9817>}
 FACIAL_R_ForeheadIn {parent=FACIAL_C_FacialRoot, loc=<-1.4719 -3.4931 9.6978>, rot=<-0.0607 0.0695 -0.0043 -0.9957>}
 FACIAL_C_Forehead {parent=FACIAL_C_FacialRoot, loc=<-0.0004 -3.5696 9.8603>, rot=<-0.0435 -0.0001 0 -0.9991>}
 FACIAL_L_ForeheadIn {parent=FACIAL_C_FacialRoot, loc=<1.4707 -3.4923 9.6982>, rot=<-0.0607 -0.0697 0.0042 -0.9957>}
 deform_hair_mb_l {parent=head, loc=<1.792 -3.7503 -5.7148>, rot=<0.1286 -0.8274 -0.5416 -0.0741>}
 dyn_hair_mb_l {parent=deform_hair_mb_l, loc=<0 0 0>, rot=<0 0 0 1>}
 dyn_hair_1_mb_l {parent=dyn_hair_mb_l, loc=<6.7334 0 0>, rot=<0 0 0 1>}
 deform_hair_m_r {parent=head, loc=<1.8928 -1.0847 6.2281>, rot=<0.064 -0.7029 0.7052 0.0673>}
 dyn_hair_m_r {parent=deform_hair_m_r, loc=<0 0 0>, rot=<0 0 0 -1>}
 dyn_hair_1_m_r {parent=dyn_hair_m_r, loc=<6.7335 0 0>, rot=<0 0 0 -1>}
 deform_hair_mb_r {parent=head, loc=<1.7918 -3.7503 5.7148>, rot=<0.1286 -0.8274 0.5416 0.0741>}
 dyn_hair_mb_r {parent=deform_hair_mb_r, loc=<0 0 0>, rot=<0 0 0 -1>}
 dyn_hair_1_mb_r {parent=dyn_hair_mb_r, loc=<6.7335 0.0001 0>, rot=<0 0 0 -1>}
 dyn_hair_1_mb_r_ctrl {controlType=EULER_TRANSFORM, parent=dyn_hair_1_mb_r, type=Control, loc=<-1.7622 -5.4981 -3.041>, rot=<0 0 0 1>}
 deform_hat_bk {parent=head, loc=<7.7042 -0.5992 0>, rot=<0 0 0.9937 -0.112>}
 deform_hat_bk1 {parent=deform_hat_bk, loc=<1.6126 8.2187 0>, rot=<0.9982 0.0602 0 0>}
 deform_hat_b_l {parent=deform_hat_bk, loc=<2.0547 6.3254 -5.0227>, rot=<0.962 0.0301 -0.061 0.2645>}
 deform_hat_m_l {parent=deform_hat_bk, loc=<1.8798 0.0785 -8.4213>, rot=<-0.7056 0.0151 0.1458 -0.6932>}
 deform_hat_f_l {parent=deform_hat_bk, loc=<1.5307 -5.7593 -8.737>, rot=<-0.6182 0.0106 0.0705 -0.7828>}
 deform_hat_f_r {parent=deform_hat_bk, loc=<1.5307 -5.7593 8.737>, rot=<0.6182 -0.0106 0.0705 -0.7828>}
 deform_hat_b_r {parent=deform_hat_bk, loc=<2.0548 6.3254 5.0227>, rot=<0.962 0.0301 0.061 -0.2645>}
 deform_hat_mb_l {parent=deform_hat_bk, loc=<1.6383 3.3778 -7.0783>, rot=<-0.8366 -0.0352 0.1342 -0.5299>}
 deform_hat_m_r {parent=deform_hat_bk, loc=<1.8795 0.0785 8.4213>, rot=<0.7056 -0.0151 0.1458 -0.6932>}
 deform_hat_mb_r {parent=deform_hat_bk, loc=<1.6384 3.3778 7.0783>, rot=<0.8366 0.0352 0.1342 -0.5299>}
 head_ctrl {controlType=EULER_TRANSFORM, parent=head, type=Control, loc=<15 0 0>, rot=<0 0 0 1>}
 clavicle_l {parent=spine_05, loc=<4.3575 -0.1466 -2.3223>, rot=<0.0465 -0.7289 -0.009 -0.6829>}
 upperarm_l {parent=clavicle_l, loc=<15.5111 0 0>, rot=<0.0226 -0.3234 -0.0088 -0.9459>}
 upperarm_twist_01_l {parent=upperarm_l, loc=<0.0041 0 0>, rot=<0 0 0 -1>}
 deform_upperarm_bk_base_l {parent=upperarm_twist_01_l, loc=<1.0813 -7.2193 -1.9453>, rot=<0 0 0 1>}
 deform_upperarm_bk_l {parent=deform_upperarm_bk_base_l, loc=<0 0 0>, rot=<0 0 0 1>}
 deform_upperarm_fr_base_l {parent=upperarm_twist_01_l, loc=<-0.3691 7.0506 -2.765>, rot=<0 0 0 1>}
 deform_upperarm_fr_l {parent=deform_upperarm_fr_base_l, loc=<0 0 0>, rot=<0 0 0 1>}
 deform_upperarm_up_base_l {parent=upperarm_twist_01_l, loc=<-3.2216 -1.1391 3.3311>, rot=<0 0 0 1>}
 deform_upperarm_up_l {parent=deform_upperarm_up_base_l, loc=<0 0 0>, rot=<0 0 0 1>}
 deform_shoulderpad_l {parent=upperarm_twist_01_l, loc=<0 0 0>, rot=<0.7071 0 0 -0.7071>}
 dyn_shoulderpad_l {parent=deform_shoulderpad_l, loc=<0 0 0>, rot=<0 0 0 1>}
 dyn_shoulderpad_1_l {parent=dyn_shoulderpad_l, loc=<-8.6796 -3.5244 -1.237>, rot=<-0.0067 0.0042 0.283 -0.9591>}
 clavicle_r {parent=spine_05, loc=<4.3575 -0.1466 2.3223>, rot=<0.7289 0.0465 -0.6829 0.009>}
 upperarm_r {parent=clavicle_r, loc=<-15.5111 0 0>, rot=<0.0226 -0.3234 -0.0088 -0.9459>}
 upperarm_twist_01_r {parent=upperarm_r, loc=<-0.0041 0 0>, rot=<0 0 0 -1>}
 deform_upperarm_bk_base_r {parent=upperarm_twist_01_r, loc=<-1.0814 7.2192 1.9455>, rot=<0 0 0 1>}
 deform_upperarm_bk_r {parent=deform_upperarm_bk_base_r, loc=<0 0 0>, rot=<0 0 0 1>}
 deform_upperarm_fr_base_r {parent=upperarm_twist_01_r, loc=<0.3689 -7.0506 2.7652>, rot=<0 0 0 1>}
 deform_upperarm_fr_r {parent=deform_upperarm_fr_base_r, loc=<0 0 0>, rot=<0 0 0 1>}
 deform_upperarm_up_base_r {parent=upperarm_twist_01_r, loc=<3.2214 1.1391 -3.3308>, rot=<0 0 0 1>}
 deform_upperarm_up_r {parent=deform_upperarm_up_base_r, loc=<0 0 0>, rot=<0 0 0 -1>}
 deform_shoulderpad_r {parent=upperarm_twist_01_r, loc=<-0.0003 0 0.0003>, rot=<0 -0.7071 -0.7071 0>}
 dyn_shoulderpad_r {parent=deform_shoulderpad_r, loc=<0 0 0>, rot=<0 0 0 1>}
 dyn_shoulderpad_1_r {parent=dyn_shoulderpad_r, loc=<-8.6798 -3.5245 1.237>, rot=<0.0067 -0.0042 0.283 -0.9591>}
 deform_pec_aim_base_l {parent=spine_04, loc=<7.8732 12.6388 -2.4123>, rot=<0.0102 -0.4602 0.3979 -0.7936>}
 deform_pec_aim_l {parent=deform_pec_aim_base_l, loc=<0 0 0>, rot=<0 0 0 1>}
 deform_pec_base_l {parent=deform_pec_aim_l, loc=<10.766 0 0>, rot=<0 0 0 1>}
 deform_pec_l {parent=deform_pec_base_l, loc=<0.055 0 0>, rot=<0 0 0 1>}
 deform_pec_aim_base_r {parent=spine_04, loc=<7.8727 12.6388 2.4123>, rot=<-0.4602 -0.0102 0.7936 0.3979>}
 deform_pec_aim_r {parent=deform_pec_aim_base_r, loc=<0 0 0>, rot=<0 0 0 1>}
 deform_pec_base_r {parent=deform_pec_aim_r, loc=<-10.7662 -0.0006 -0.0007>, rot=<0 0 0 1>}
 deform_pec_r {parent=deform_pec_base_r, loc=<0 0 0>, rot=<0 0 0 1>}
]
variables [is_in_frontend {type=bool, default=False}]
