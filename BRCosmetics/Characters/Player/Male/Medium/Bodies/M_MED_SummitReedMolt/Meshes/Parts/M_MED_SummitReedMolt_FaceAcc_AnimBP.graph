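# Animation Blueprint for the SummitReedMolt character's face accessory, driving secondary motion on its hair with and without a hat.
# It copies parent and body poses into a Control Rig, then RigidBody simulations weighted by hat state and a post Control Rig.
# Additive bone corrections follow per hat state, with constraints pulling back hair toward spine and clavicles when a backpack is worn.
# Events derive hat state from face-material parameters and update backpack, emote, gravity and physics-setting variables.
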AnimBlueprint /BRCosmetics/Characters/Player/Male/Medium/Bodies/M_MED_SummitReedMolt/Meshes/Parts/M_MED_SummitReedMolt_FaceAcc_AnimBP
target_skeleton=&/M_MED_SummitReedMolt_FaceAcc_Skeleton
graphs [
 AnimGraph {nodes [
   n1 = "Constraint - dyn_hair_1_bk (P)"(ComponentPose=Local To Component(LocalPose=Use cached pose 'prebackpack_hat'()), Alpha=0.3, ConstraintWeights_0=1) {
    alphaScaleBiasClamp {bClampResult=true, clampMax=0.9, clampMin=0.1}
    boneToModify {boneName=dyn_hair_1_bk}
    constraintSetup [{targetBone {boneName=spine_04}}]
    constraintWeights [1]
   }
   n2 = "Constraint - dyn_hair_1_b_r (PP)"!(ComponentPose=@n1, Alpha=0.9, ConstraintWeights_0=0.7, ConstraintWeights_1:float=0.3) -> (Pose) {
    alphaScaleBiasClamp {bClampResult=true, clampMax=0.9, clampMin=0.1}
    boneToModify {boneName=dyn_hair_1_b_r}
    constraintSetup [{targetBone {boneName=spine_04}}, {targetBone {boneName=clavicle_l}}]
    constraintWeights [1,1]
   }
   n3 = "Constraint - dyn_hair_1_b_l (PP)"!(ComponentPose=@n2, Alpha=0.3, ConstraintWeights_0=0.7, ConstraintWeights_1:float=0.3) -> (Pose) {
    alphaScaleBiasClamp {bClampResult=true, clampMax=0.9, clampMin=0.1}
    boneToModify {boneName=dyn_hair_1_b_l}
    constraintSetup [{targetBone {boneName=spine_04}}, {targetBone {boneName=clavicle_r}}]
    constraintWeights [1,1]
   }
   n4 = Component To Local(ComponentPose=@n3)
   n5 = Use cached pose 'prebackpack_hat'()
   n6 = Get IsWearingBackpack() -> (IsWearingBackpack:bool)
   n7 = "Constraint - dyn_hair_1_bk (P)"(ComponentPose=Local To Component(LocalPose=Use cached pose 'prebackpack'()), Alpha=0.4, ConstraintWeights_0=1) {
    alphaScaleBiasClamp {bClampResult=true, clampMax=0.8, clampMin=0.2}
    boneToModify {boneName=dyn_hair_1_bk}
    constraintSetup [{targetBone {boneName=spine_04}}]
    constraintWeights [1]
   }
   n8 = "Constraint - dyn_hair_1_b_r (PP)"(ComponentPose=@n7, Alpha=0.6, ConstraintWeights_0=0.7, ConstraintWeights_1=0.3) {
    alphaScaleBiasClamp {bClampResult=true, clampMax=0.8, clampMin=0.2}
    boneToModify {boneName=dyn_hair_1_b_r}
    constraintSetup [{targetBone {boneName=spine_04}}, {targetBone {boneName=clavicle_l}}]
    constraintWeights [1,1]
   }
   n9 = Component To Local(ComponentPose=@n8)
   n10 = Use cached pose 'prebackpack'()
   n11 = Get IsWearingBackpack()
   Output Pose(Result=Blend Poses by bool(
     BlendPose_0=Blend Poses by bool(BlendPose_0=@n4, BlendPose_1=@n5, bActiveValue=@n6)
     BlendPose_1=Blend Poses by bool(BlendPose_0=@n9, BlendPose_1=@n10, bActiveValue=@n11)
     bActiveValue=Get Hat_Bool() -> (Hat_Bool:bool)
    )) {layerGroup=None}
   n12 = Get Body_Mesh()
   n13 = Control Rig!(
    Source=Layered blend per bone(BasePose=Copy Pose From Mesh() {bUseAttachedParent=true}, BlendPoses_0=Copy Pose From Mesh(SourceMeshComponent=@n12)) {layerSetup [{branchFilters [{blendDepth=5, boneName=clavicle_l}, {blendDepth=5, boneName=clavicle_r}]}]}
    Alpha
    IsInHat:double=Get Hat() -> (Hat:double)
   ) -> (Pose) {controlRigAssetReference {blueprintRigClass=&/M_MED_SummitReedMolt_FaceAcc_CtrlRig.M_MED_SummitReedMolt_FaceAcc_CtrlRig_C}}
   Save cached pose 'Main'(Pose=@n13) {cacheName=Main}
   n14 = Get Hat()
   n15 = RigidBody!(
    ComponentPose=Local To Component(LocalPose=Use cached pose 'Main'())
    Alpha=@n14
    OverrideWorldGravity:Vector="0,0,-980"
    SimSpaceSettings:Sim Space Settings Structure=(WorldAlpha=0.5,VelocityScaleZ=0.63,DampingAlpha=1,MaxLinearVelocity=10000,MaxAngularVelocity=10000,MaxLinearAcceleration=3500,MaxAngularAcceleration=2500,ExternalLinearDragV=(X=1,Y=1,Z=1),ExternalLinearVelocity=(X=0,Y=0,Z=0),ExternalAngularVelocity=(X=0,Y=0,Z=0))
   ) -> (Pose) {
    bClampLinearTranslationLimitToRefPose=true
    bOverrideWorldGravity=true
    overridePhysicsAsset=&/M_MED_SummitReedMolt_FaceAcc_Physics
    simulationSpace=BaseBoneSpace
    alphaScaleBias {bias=1, scale=-1}
    baseBoneRef {boneName=neck_02}
    overrideWorldGravity {z=-980}
    simSpaceSettings {maxAngularAcceleration=2500, maxLinearAcceleration=3500, velocityScaleZ=0.63, worldAlpha=0.5, externalLinearDragV=<1 1 1>}
   }
   n16 = RigidBody(
    ComponentPose=@n15
    Alpha=@n14
    OverrideWorldGravity="0,0,-980"
    SimSpaceSettings=(WorldAlpha=0.2,VelocityScaleZ=0.1,DampingAlpha=1,MaxLinearVelocity=10000,MaxAngularVelocity=10000,MaxLinearAcceleration=2500,MaxAngularAcceleration=2500,ExternalLinearDragV=(X=0.2,Y=0.2,Z=0.2),ExternalLinearVelocity=(X=0,Y=0,Z=0),ExternalAngularVelocity=(X=0,Y=0,Z=0))
   ) {
    bClampLinearTranslationLimitToRefPose=true
    bOverrideWorldGravity=true
    overridePhysicsAsset=&/M_MED_SummitReedMolt_FaceAcc_Hat_Physics
    simulationSpace=BaseBoneSpace
    baseBoneRef {boneName=neck_02}
    overrideWorldGravity {z=-980}
    simSpaceSettings {maxAngularAcceleration=2500, maxLinearAcceleration=2500, velocityScaleZ=0.1, worldAlpha=0.2, externalLinearDragV=<0.2 0.2 0.2>}
   }
   n17 = RigidBody(
    ComponentPose=Local To Component(LocalPose=Use cached pose 'Main'())
    Alpha=@n14
    OverrideWorldGravity="0,0,-980"
    SimSpaceSettings=(WorldAlpha=0.5,VelocityScaleZ=0.63,DampingAlpha=1,MaxLinearVelocity=10000,MaxAngularVelocity=10000,MaxLinearAcceleration=3500,MaxAngularAcceleration=2500,ExternalLinearDragV=(X=1,Y=1,Z=1),ExternalLinearVelocity=(X=0,Y=0,Z=0),ExternalAngularVelocity=(X=0,Y=0,Z=0))
   ) {
    bClampLinearTranslationLimitToRefPose=true
    bOverrideWorldGravity=true
    overridePhysicsAsset=&/M_MED_SummitReedMolt_FaceAcc_Physics1
    simulationSpace=BaseBoneSpace
    alphaScaleBias {bias=1, scale=-1}
    baseBoneRef {boneName=neck_02}
    overrideWorldGravity {z=-980}
    simSpaceSettings {maxAngularAcceleration=2500, maxLinearAcceleration=3500, velocityScaleZ=0.63, worldAlpha=0.5, externalLinearDragV=<1 1 1>}
   }
   n18 = RigidBody(
    ComponentPose=@n17
    Alpha=@n14
    OverrideWorldGravity="0,0,-980"
    SimSpaceSettings=(WorldAlpha=0.2,VelocityScaleZ=0.1,DampingAlpha=1,MaxLinearVelocity=10000,MaxAngularVelocity=10000,MaxLinearAcceleration=2500,MaxAngularAcceleration=2500,ExternalLinearDragV=(X=0.2,Y=0.2,Z=0.2),ExternalLinearVelocity=(X=0,Y=0,Z=0),ExternalAngularVelocity=(X=0,Y=0,Z=0))
   ) {
    bClampLinearTranslationLimitToRefPose=true
    bOverrideWorldGravity=true
    overridePhysicsAsset=&/M_MED_SummitReedMolt_FaceAcc_Hat_Physics1
    simulationSpace=BaseBoneSpace
    baseBoneRef {boneName=neck_02}
    overrideWorldGravity {z=-980}
    simSpaceSettings {maxAngularAcceleration=2500, maxLinearAcceleration=2500, velocityScaleZ=0.1, worldAlpha=0.2, externalLinearDragV=<0.2 0.2 0.2>}
   }
   n19 = Control Rig!(
    Source=Blend Poses by bool(BlendPose_0=Component To Local(ComponentPose=@n18), BlendPose_1=Component To Local(ComponentPose=@n16))
    Alpha=@n14
    is_in_frontend:bool=Get bIsInFrontEnd()
   ) -> (Pose) {controlRigAssetReference {blueprintRigClass=&/M_MED_SummitReedMolt_FaceAcc_Post_CtrlRig.M_MED_SummitReedMolt_FaceAcc_Post_CtrlRig_C}}
   Save cached pose 'PreEmoteCorrections'(Pose=@n19) {cacheName=PreEmoteCorrections}
   n20 = "Transform (Modify) Bone - Bone: dyn_hair_bk"!(ComponentPose=Local To Component(LocalPose=Use cached pose 'PreEmoteCorrections'()), bAlphaBoolEnabled:bool=True, Rotation="0,-10,0") -> (Pose) {
    alphaInputType=Bool
    rotationMode=BMM_Additive
    rotationSpace=BCS_BoneSpace
    translationSpace=BCS_BoneSpace
    boneToModify {boneName=dyn_hair_bk}
    rotation {yaw=-10}
   }
   n21 = "Transform (Modify) Bone - Bone: dyn_hair_1_bk"!(ComponentPose=@n20, bAlphaBoolEnabled:bool=True, Rotation="0,20,0") -> (Pose) {
    alphaInputType=Bool
    rotationMode=BMM_Additive
    rotationSpace=BCS_BoneSpace
    translationSpace=BCS_BoneSpace
    boneToModify {boneName=dyn_hair_1_bk}
    rotation {yaw=20}
   }
   n22 = "Transform (Modify) Bone - Bone: dyn_hair_1_b_l"!(ComponentPose=@n21, bAlphaBoolEnabled:bool=True, Rotation="0,10,10") -> (Pose) {
    alphaInputType=Bool
    rotationMode=BMM_Additive
    rotationSpace=BCS_BoneSpace
    translationSpace=BCS_BoneSpace
    boneToModify {boneName=dyn_hair_1_b_l}
    rotation {roll=10, yaw=10}
   }
   Save cached pose 'prebackpack'(Pose=Component To Local(ComponentPose=@n22)) {cacheName=prebackpack}
   n23 = "Transform (Modify) Bone - Bone: dyn_hair_bk"(ComponentPose=Local To Component(LocalPose=Use cached pose 'PreEmoteCorrections'()), bAlphaBoolEnabled=True, Rotation="0,-10,0") {
    alphaInputType=Bool
    rotationMode=BMM_Additive
    rotationSpace=BCS_BoneSpace
    translationSpace=BCS_BoneSpace
    boneToModify {boneName=dyn_hair_bk}
    rotation {yaw=-10}
   }
   n24 = "Transform (Modify) Bone - Bone: dyn_hair_1_bk"(ComponentPose=@n23, bAlphaBoolEnabled=True, Rotation="0,30,0") {
    alphaInputType=Bool
    rotationMode=BMM_Additive
    rotationSpace=BCS_BoneSpace
    translationSpace=BCS_BoneSpace
    boneToModify {boneName=dyn_hair_1_bk}
    rotation {yaw=30}
   }
   n25 = "Transform (Modify) Bone - Bone: dyn_hair_1_b_l"(ComponentPose=@n24, bAlphaBoolEnabled=True, Rotation="0,0,20") {
    alphaInputType=Bool
    rotationMode=BMM_Additive
    rotationSpace=BCS_BoneSpace
    translationSpace=BCS_BoneSpace
    boneToModify {boneName=dyn_hair_1_b_l}
    rotation {roll=20}
   }
   n26 = "Transform (Modify) Bone - Bone: dyn_hair_b_l"!(ComponentPose=@n25, bAlphaBoolEnabled:bool=True, Rotation="0,10,0") -> (Pose) {
    alphaInputType=Bool
    rotationMode=BMM_Additive
    rotationSpace=BCS_BoneSpace
    translationSpace=BCS_BoneSpace
    boneToModify {boneName=dyn_hair_b_l}
    rotation {yaw=10}
   }
   Save cached pose 'prebackpack_hat'(Pose=Component To Local(ComponentPose=@n26)) {cacheName=prebackpack_hat}
  ]}
 EventGraph {nodes [
   Event BlueprintUpdateAnimation()
   > n1 = "Gravity Override State Function (Parameter Asset) (Nativized)"(
    CharacterPartAnimInstance=Self-Reference()
    GameTimeInSeconds=GetGameTimeInSeconds()
    Parameters=/BRCosmetics/Characters/Player/Male/Medium/Bodies/M_MED_GnatGala/Meshes/M_MED_GnatGala_Parameters
   )
   > Set Gravity Override(Gravity Override=@n1)
   > n2 = Sequence()
   > n3 = Branch(Condition=Get Hat_Bool())
   > n4 = Branch(Condition=Get bIsInFrontEnd())
   > Set!(
    execute
    RBAN Settings Out_ComponentLinearAccScale:Vector="0,0,0"
    RBAN Settings Out_ComponentLinearVelScale:Vector="0,0,0"
    RBAN Settings Out_SimSpaceSettings_WorldAlpha:float=0.6
    RBAN Settings Out_SimSpaceSettings_VelocityScaleZ:float=0.6
    RBAN Settings Out_SimSpaceSettings_DampingAlpha:float=1
    RBAN Settings Out_SimSpaceSettings_MaxLinearVelocity:float=10000
    RBAN Settings Out_SimSpaceSettings_MaxAngularVelocity:float=10000
    RBAN Settings Out_SimSpaceSettings_MaxLinearAcceleration:float=5000
    RBAN Settings Out_SimSpaceSettings_MaxAngularAcceleration:float=2500
    RBAN Settings Out_SimSpaceSettings_ExternalLinearDragV:Vector="1,1,1"
    RBAN Settings Out_SimSpaceSettings_ExternalLinearVelocity:Vector="0,0,0"
    RBAN Settings Out_SimSpaceSettings_ExternalAngularVelocity:Vector="0,0,0"
   ) -> (then, Output_Get:Fort Rigid Body Anim Node Parameters Structure)
   Set(
    execute=@n4.else
    RBAN Settings Out_ComponentLinearAccScale="0,0,0"
    RBAN Settings Out_ComponentLinearVelScale="0,0,0"
    RBAN Settings Out_SimSpaceSettings_WorldAlpha=0.2
    RBAN Settings Out_SimSpaceSettings_VelocityScaleZ=0.1
    RBAN Settings Out_SimSpaceSettings_DampingAlpha=1
    RBAN Settings Out_SimSpaceSettings_MaxLinearVelocity=10000
    RBAN Settings Out_SimSpaceSettings_MaxAngularVelocity=10000
    RBAN Settings Out_SimSpaceSettings_MaxLinearAcceleration=2500
    RBAN Settings Out_SimSpaceSettings_MaxAngularAcceleration=2500
    RBAN Settings Out_SimSpaceSettings_ExternalLinearDragV="0.2,0.2,0.2"
    RBAN Settings Out_SimSpaceSettings_ExternalLinearVelocity="0,0,0"
    RBAN Settings Out_SimSpaceSettings_ExternalAngularVelocity="0,0,0"
   )
   n5 = Branch(execute=@n3.else, Condition=Get bIsInFrontEnd())
   > Set(
    RBAN Settings Out_ComponentLinearAccScale="0,0,0"
    RBAN Settings Out_ComponentLinearVelScale="0,0,0"
    RBAN Settings Out_SimSpaceSettings_WorldAlpha=0.7
    RBAN Settings Out_SimSpaceSettings_VelocityScaleZ=1
    RBAN Settings Out_SimSpaceSettings_DampingAlpha=1
    RBAN Settings Out_SimSpaceSettings_MaxLinearVelocity=6000
    RBAN Settings Out_SimSpaceSettings_MaxAngularVelocity=6000
    RBAN Settings Out_SimSpaceSettings_MaxLinearAcceleration=5000
    RBAN Settings Out_SimSpaceSettings_MaxAngularAcceleration=2500
    RBAN Settings Out_SimSpaceSettings_ExternalLinearDragV="1,1,1"
    RBAN Settings Out_SimSpaceSettings_ExternalLinearVelocity="0,0,0"
    RBAN Settings Out_SimSpaceSettings_ExternalAngularVelocity="0,0,0"
   )
   Set(
    execute=@n5.else
    RBAN Settings Out_ComponentLinearAccScale="0,0,0"
    RBAN Settings Out_ComponentLinearVelScale="0,0,0"
    RBAN Settings Out_SimSpaceSettings_WorldAlpha=0.5
    RBAN Settings Out_SimSpaceSettings_VelocityScaleZ=0.63
    RBAN Settings Out_SimSpaceSettings_DampingAlpha=1
    RBAN Settings Out_SimSpaceSettings_MaxLinearVelocity=10000
    RBAN Settings Out_SimSpaceSettings_MaxAngularVelocity=10000
    RBAN Settings Out_SimSpaceSettings_MaxLinearAcceleration=3500
    RBAN Settings Out_SimSpaceSettings_MaxAngularAcceleration=2500
    RBAN Settings Out_SimSpaceSettings_ExternalLinearDragV="1,1,1"
    RBAN Settings Out_SimSpaceSettings_ExternalLinearVelocity="0,0,0"
    RBAN Settings Out_SimSpaceSettings_ExternalAngularVelocity="0,0,0"
   )
   n6 = Branch(execute=@n2.then_1, Condition=Get bIsPlayingEmote())
   n7 = GetAnimInstance(self=GetOwnerSkeletalMeshComponent())
   > Set Emote_ReedGrit(Emote_ReedGrit:bool="float >= float"(A=GetCurveValue(self=@n7, CurveName=emote_BuiltIn), B=2)) -> (Output_Get:bool)
   Set Emote_ReedGrit(execute=@n6.else, Emote_ReedGrit=false)
   n8 = Event OnCharacterCustomizationLoaded()
   > Set Body_Mesh(Body_Mesh=Get Array Item with EFort Custom Part Type(Array=@n8.CustomMeshParts))
   > n9 = Get Dynamic Material Instance(SkeletalMeshComponent=Get Array Item with EFort Custom Part Type(Array=@n8.CustomMeshParts, PartType=Face), ElementIndex=1)
   > n10 = Get Scalar Parameter Value!(execute, self:Material Instance Dynamic Object Reference=@n9, ParameterName:Name=Hide Element 01) -> (then, ReturnValue:float) {is=title K2Node_CallFunction}
   > n11 = Get Scalar Parameter Value(self=@n9, ParameterName=Hide Element 02)
   > Set Bandana(Bandana:double=float - float(A=1, B=@n11)) -> (Output_Get:double)
   > n12 = Set Hat(Hat:double=float - float(A=1, B=@n10)) -> (Output_Get:double)
   > Set Hat_Bool(Hat_Bool:bool="float >= float"(A=@n12, B=1)) -> (Output_Get:bool)
   n13 = CheckForBackbling()
   > Set IsWearingBackpack(IsWearingBackpack:bool=OR Boolean(A=@n13, B=@n13.bIsBackblingACape)) -> (Output_Get:bool)
  ]}
]
variables [
 Bandana {type=number}
 Body_Mesh {type=SkeletalMeshComponent}
 Emote_ReedGrit {type=boolean}
 Gravity Override {type=Vector}
 Hat {type=number}
 Hat_Bool {type=boolean}
 IsWearingBackpack {type=boolean}
 RBAN Settings Out {
  type=FortRigidBodyAnimNodeParameters
  default {simSpaceSettings {maxAngularAcceleration=2500, maxLinearAcceleration=2500, velocityScaleZ=0.1, worldAlpha=0.2, externalLinearDragV=<0.2 0.2 0.2>}}
 }
]
